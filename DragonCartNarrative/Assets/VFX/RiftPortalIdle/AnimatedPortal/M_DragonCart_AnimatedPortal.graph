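# Unlit, translucent, two-sided animated rift portal material for Niagara mesh particles in DragonCart, with a PortalInterior static switch.
# The interior projects a tinted nebula texture along the camera vector, adds a color gradient, and uses a circular mask sized by the PortalOpen dynamic parameter.
# Otherwise, two panning cloud noises drive colors, opacity and a sharpened band, with normal-mapped Fresnel and HDR-boosted emissive.
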Material /DragonCartNarrative/Assets/VFX/RiftPortalIdle/AnimatedPortal/M_DragonCart_AnimatedPortal
graph {
 properties {
  bAutomaticallySetUsageInEditor=false
  bUsedWithNiagaraMeshParticles=true
  blendMode=BLEND_Translucent
  shadingModel=MSM_Unlit
  twoSided=true
  usedShadingModels=Unlit
 }
 nodes [
  n1 = NamedRerouteUsage()
  LinearInterpolate(A=@n1, B=ComponentMask(None=TextureCoordinate()) {g=true}) {constAlpha=0.7}
  MaterialFunctionCall(Speed=Constant() {r=0.1}) {materialFunction=&DebugTimeSine}
  n2 = TransformPosition(None=WorldPosition().XYZ) {transformSourceType=TRANSFORMPOSSOURCE_World}
  n3 = Constant() {r=10000}
  MaterialFunctionCall(Vector or UVs=ComponentMask(None=Divide(A=@n2, B=@n3) {constB=5000}) {g=true, r=true}) {materialFunction=&VectorToRadialValue}
  n4 = ScalarParameter() {defaultValue=5000, parameterName=PortalMaxSize}
  n5 = NamedRerouteDeclaration(None=DynamicParameter() {paramNames [PortalOpen,Param2,Param3,Param4]}.PortalOpen) {name=PortalOpen, nodeColor=<rgba 0.2656 1 0 1>}
  n6 = MaterialFunctionCall!(
   MaskPosition=ActorPositionWS()
   MaskDirection=Transform(None=Constant3Vector() {constant=<rgba 0 0 -1 0>}) {transformSourceType=TRANSFORMSOURCE_Local}
   MaskRadius=Add(A=Multiply(A=@n4, B=@n5) {constA=5000}) {constB=200}
   MaskEdgeGradientSize
   MaskEdgeGradientOffset
   MaskEdgeGradientPower
  ) {materialFunction=/Game/Effects/Niagara/MasterMaterials/Functions/Opacity/MF_FX_Circular_Portal_Mask}
  n7 = VertexColor()
  n8 = NamedRerouteUsage()
  n9 = Constant2Vector() {g=0.5, r=1}
  n10 = NamedRerouteUsage()
  n11 = Add(
   A=Multiply(A=NamedRerouteUsage(), B=Constant2Vector() {g=2}) {constB=2}
   B=MaterialFunctionCall(UVs=TextureCoordinate(), Texture Scale=LinearInterpolate(A=@n9, Alpha=@n10)) {materialFunction=&ScaleUVsByCenter}.UVs
  )
  n12 = VectorParameter() {parameterName=TileAndPan1, defaultValue=<rgba 5 0.2 -0.01 0.1>}
  n13 = Panner(Coordinate=Multiply(A=@n11, B=ComponentMask(None=@n12.RGB) {g=true, r=true}), Speed=ComponentMask(None=@n12.RGBA) {a=true, b=true})
  n14 = TextureSample(UVs=@n13) {texture=/Game/Weapons/WeaponSkins/Wraps/Textures/RadishMem/T_Wrap_Radish_CloudNoise}
  n15 = VectorParameter() {parameterName=TileAndPan2, defaultValue=<rgba 7 0.6 0.01 0.15>}
  n16 = ComponentMask(None=@n15.RGB) {g=true, r=true}
  n17 = TextureSample(UVs=Panner(Coordinate=Multiply(A=@n11, B=@n16), Speed=ComponentMask(None=@n15.RGBA) {a=true, b=true})) {texture=&T_Wrap_Radish_CloudNoise}
  n18 = Multiply(A=Saturate(None=Multiply(A=@n8) {constB=4}), B=Multiply(A=Add(A=@n14.G, B=@n17.G)) {constB=2})
  n19 = StaticSwitchParameter(
   True=Multiply(A=LinearInterpolate(B=@n6.OpacityMask, Alpha=@n7.R) {constA=1}, B=NamedRerouteUsage())
   False=Multiply(A=Max(A=@n8, B=@n18), B=NamedRerouteUsage())
  ) {parameterName=PortalInterior}
  MaterialFunctionCall(Alpha Threshold=@n19) {materialFunction=&DitherTemporalAA_Fortnite}
  n20 = ComponentMask(None=TextureCoordinate()) {g=true}
  n21 = Multiply(A=VectorParameter() {parameterName=Color4, defaultValue=<rgba 0.5 0.1407 0 1>}.RGB, B=Power(Base=@n20) {constExponent=6})
  n22 = TextureObjectParameter() {parameterName=InteriorTexture, texture=/Game/Characters/Player/Male/Medium/Bodies/M_Med_Soldier_04/Skins/Galaxy/Textures/T_FN_Nebula_style2}
  n23 = Multiply(
   A=MaterialFunctionCall!(CameraVector, Texture=@n22, Tiling=ScalarParameter() {defaultValue=2, parameterName=InteriorScale}, Offset, AxisFadeContrast) {materialFunction=/Engine/Functions/Engine_MaterialFunctions01/Texturing/TriplanarCameraVector}.Result
   B=VectorParameter() {parameterName=InteriorTint, defaultValue=<rgba 0.5888 0.7133 0.8 1>}.RGB
  )
  n24 = ComponentMask(None=TextureCoordinate()) {g=true}
  n25 = Multiply(A=@n17.G) {constB=0.5}
  n26 = LinearInterpolate(
   A=Multiply(A=Multiply(A=@n14.G, B=@n17.G), B=NamedRerouteUsage())
   B=Multiply(A=NamedRerouteUsage(), B=Add(A=@n14.G, B=@n25))
   Alpha=NamedRerouteUsage()
  )
  n27 = SmoothStep(Value=ComponentMask(None=LinearInterpolate(A=@n24, B=@n26) {constAlpha=0.25}) {r=true}) {constMax=0.7, constMin=0.6, constValue=0.5}
  n28 = Power(Base=Saturate(None=Sine(None=@n27))) {constExponent=6}
  n29 = NamedRerouteUsage()
  n30 = Power(Base=@n26) {constExponent=1.5}
  n31 = Multiply(
   A=Transform(None=TextureSample(UVs=@n13) {samplerType=SAMPLERTYPE_Normal, texture=/Game/Effects/Fort_Effects/Creatures/Elemental/Ice/TilingCloudNormalsMap}.RGB)
   B=TwoSidedSign()
  )
  n32 = Add(
   A=Add(A=Multiply(A=@n28, B=@n29), B=Multiply(A=@n30) {constB=0.6})
   B=Multiply(
    A=Fresnel(Normal=@n31) {baseReflectFraction=0.2, exponent=2}
    B=VectorParameter() {parameterName=Color3, defaultValue=<rgba 0.3 0.078 0.0297 1>}.RGB
   )
  )
  n33 = Saturate(None=@n32)
  n34 = LinearInterpolate(Alpha=Power(Base=@n1) {constExponent=3}) {constA=2, constB=15}
  MaterialFunctionCall(Emissive=StaticSwitchParameter(
    True=Add(A=@n21, B=@n23)
    False=MaterialFunctionCall(LDR Input=@n33, HDR Tint=@n32, HDR Intensity Multiplier=@n34) {materialFunction=&DeriveHDRfromLDR}.RemappedRange
   ) {parameterName=PortalInterior}) {materialFunction=&MF_FX_EmissiveAdjustments}
  n35 = Multiply(
   A=TransformPosition(None=WorldPosition().XYZ) {transformSourceType=TRANSFORMPOSSOURCE_World}
   B=Constant3Vector() {constant=<rgba 1 1 0.1 0>}
  )
  n36 = NamedRerouteUsage()
  Multiply(
   A=Transform(None=Normalize(VectorInput=@n35)) {transformSourceType=TRANSFORMSOURCE_Local}
   B=Multiply(A=OneMinus(None=@n36), B=ScalarParameter() {defaultValue=-100})
  )
  Multiply(A=@n19) {constB=2}
  NamedRerouteDeclaration(None=VectorParameter() {parameterName=Color2, defaultValue=<rgba 1 0.5016 0.1198 1>}.RGB) {name=Color2, nodeColor=<rgba 1 0.3873 0 1>}
  NamedRerouteDeclaration(None=VectorParameter() {parameterName=Color1, defaultValue=<rgba 1 0.7782 0 1>}.RGB) {name=Color1, nodeColor=<rgba 1 0.3873 0 1>}
  n37 = Constant() {r=0.5}
  NamedRerouteDeclaration(None=SphereMask(A=Power(Base=@n24) {constExponent=2.5}, B=@n37, Radius=@n37) {hardnessPercent=0}) {name=SphereMask, nodeColor=<rgba 0.6406 1 0 1>}
  NamedRerouteDeclaration(None=OneMinus(None=@n27)) {name=InnerOpacityMask, nodeColor=<rgba 1 0 0.0703 1>}
  n38 = OneMinus(None=@n20)
  NamedRerouteDeclaration(None=Saturate(None=Multiply(A=@n38) {constB=9})) {name=InnerPortalMask, nodeColor=<rgba 0 1 0.3906 1>}
  Saturate(None=@n6.OpacityMask)
  ScalarParameter() {defaultValue=1, parameterName=Param_4}
 ]
}
