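# Niagara system for the DragonCart animated rift portal, combining a portal mesh with flare, cracks, ring lights, shockwaves, shards and GPU sparkles.
# A PortalOpen value, read from system and NPC_DragonCart_Orb_Rift collection parameters, drives portal scale, crack materials, light radius and spawn radius.
# Portal, cracks and lights persist beyond their lifetime, while waves, shards and sparkles expand and fade.
# User.RedRift switches wave, shard and sparkle colors between redder and more orange palettes.
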NiagaraSystem /DragonCartNarrative/Assets/VFX/RiftPortalIdle/AnimatedPortal/NS_DragonCart_Animated_Portal
dependencies {
 usedDataInterfaces [&NiagaraDataInterfaceCurve, &NiagaraDataInterfaceMeshRendererInfo]
 usedDynamicInputs []
 usedModules [
  &self:SystemSpawnScript.SystemScriptSource.SystemScriptGraph.NiagaraNodeAssignment_1.SetVariables_#1
  &SystemState
  &self:SystemSpawnScript.SystemScriptSource.SystemScriptGraph.NiagaraNodeAssignment_0.SetVariables_#2
 ]
}
emitters {
 Cracks {
  emitterUpdateScript [
   &EmitterState(Life Cycle Mode=::NewEnumerator1, Loop Behavior=::NewEnumerator1, Loop Duration Mode=::NewEnumerator1, UseLoopDelay=false)
   &SpawnBurst_Instantaneous(Spawn Count=4, Use Spawn Probability=false, UseLoopCountLimit=false)
  ]
  particleSpawnScript [
   &InitializeParticle(
    Color Mode=::NewEnumerator0
    Lifetime=10
    Mesh Renderer Array Visibility Mode=::NewEnumerator3
    Mesh Renderer Info=DataInterface() {propertyValues {MeshRenderer=&self:Cracks_0.NiagaraMeshRendererProperties_0}}
    Mesh Scale Max=<160 160 10>
    Mesh Scale Min=<120 120 5>
    Mesh Scale Mode=::NewEnumerator2
    Position Offset=<0 0 -250>
    UsePositionOffset=true
    Write Material Random=true
   )
   &InitialMeshOrientation(
    Mesh Orientation Mode=::NewEnumerator2
    Orientation Vector=&RotateVectorByQuaternion(
     Vector To Rotate=<1 0 0>
     Quaternion=&MakeQuaternion(
      Quaternion From=0
      Axis=<0 0 1>
      Angle=&ReturnNormalizedExecIndex(
       Normalized Index Scale=&RandomRangeFloat(Minimum=0.9, Maximum=1.1)
       Normalized Index Includes One=false
      )
      From=?
      To=?
      Angle Type=::NewEnumerator1
      Normalize Vectors=false
     )
    )
    Rotation=&RandomRangeVector(Minimum=<0 0 0>, Maximum=<0 0 0>)
    Use Rotation Vector=true
   )
   &CameraOffset(Camera Offset Amount=200)
  ]
  particleUpdateScript [
   &ParticleState(Kill Particles When Lifetime Has Elapsed=false)
   &DynamicMaterialParameters(
    Index 0 Param 1=&FloatFromCurve(
     FloatCurve=DataInterface() {propertyValues {Curve {
        defaultValue=1
        keys [
         {arriveTangent=11.5047, interpMode=RCIM_Cubic, time=0, value=0}
         {arriveTangent=1, interpMode=RCIM_Cubic, leaveTangent=1, time=0.8, value=0.4}
         {arriveTangent=3.8371, tangentMode=RCTM_User, time=1, value=1}
        ]
       }}}
     CurveIndex=$System.PortalOpen
    )
    Index 0 Param 2=&Lerp_Float(A=1, B=4, Alpha=&ReturnNormalizedExecIndex(Normalized Index Includes One=false))
    Index 0 Param 3=&Add_Float(
     A=&Lerp_Float(A=0, B=0.25, Alpha=$System.PortalOpen)
     B=&RandomRangeFloat(Minimum=0, Maximum=1)
    )
    Index 3 Param 1=&Add_Float(A=&Multiply_Float(A=$System.PortalOpen, B=5000), B=300)
    Module.Index0Float0Write=true
    Module.Index0Float1Write=true
    Module.Index0Float2Write=true
    Module.Index0Float3Write=false
    Module.Index3Float0Write=true
    Module.Index3Float1Write=false
    Module.Index3Float2Write=false
    Module.Index3Float3Write=false
    Write Parameter Index 3=::NewEnumerator1
   )
  ]
  rendererClasses [&NiagaraMeshRendererProperties]
  renderers [{rendererClass=&NiagaraMeshRendererProperties, rendererIndex=0}]
 }
 Lights_High {
  emitterUpdateScript [
   &EmitterState(
    Inactive Response=::NewEnumerator1
    Life Cycle Mode=::NewEnumerator1
    Loop Behavior=::NewEnumerator1
    Loop Duration Mode=::NewEnumerator1
    UseLoopDelay=false
   )
   &SpawnBurst_Instantaneous(Spawn Count=16, Spawn Time=0.05, Use Spawn Probability=false, UseLoopCountLimit=false)
  ]
  particleSpawnScript [
   &InitializeParticle(
    Color=<rgba 100 33.6857 0 1>
    Lifetime=5
    Sprite Size Mode=::NewEnumerator3
    Uniform Sprite Size=100
    UsePositionOffset=false
    Write Material Random=true
   )
   &ShapeLocation(
    MaskSpawnGroup=false
    Offset Mode=::NewEnumerator0
    Override Seed=false
    Ring / Disc Distribution Mode:ENiagaraSphereDistributionMode=::NewEnumerator1
    Ring Radius=4700
    Shape Primitive=::NewEnumerator4
    U Position=&ReturnNormalizedExecIndex()
    UseOwnerScale=true
   )
  ]
  particleUpdateScript [
   &ParticleState(Kill Particles When Lifetime Has Elapsed=false)
   /SetVariables(Particles.LightRadius=&Lerp_Float(A=10, B=2000, Alpha=$System.PortalOpen)) {isSetParameters=true}
  ]
  rendererClasses [&NiagaraSpriteRendererProperties, &NiagaraLightRendererProperties]
  renderers [
   {rendererClass=&NiagaraSpriteRendererProperties}
   {rendererClass=&NiagaraLightRendererProperties, rendererIndex=1}
  ]
 }
 PortalMesh {
  emitterUpdateScript [
   &EmitterState(
    Inactive Response=::NewEnumerator1
    Life Cycle Mode=::NewEnumerator1
    Loop Behavior=::NewEnumerator1
    Loop Duration Mode=::NewEnumerator1
    UseLoopDelay=false
   )
   &SpawnBurst_Instantaneous(Use Spawn Probability=false, UseLoopCountLimit=false)
  ]
  particleSpawnScript [
   &InitializeParticle(
    Color=<rgba 3 0.9354 0 0.25>
    Lifetime=10
    Mass Mode=::NewEnumerator1
    Mesh Scale Mode=::NewEnumerator3
    Mesh Uniform Scale=50
    Sprite Size=<30000 20000>
    Sprite Size Mode=::NewEnumerator4
    UsePositionOffset=false
    Write Material Random=true
   )
   &CameraOffset(Camera Offset Amount=500)
  ]
  particleUpdateScript [
   &ParticleState(Kill Particles When Lifetime Has Elapsed=false)
   /SetVariables(Particles.FlareColor=&MultiplyLinearColor(
     Linear Color=<rgba 4 1.369 0 1>
     Scale Factor=&MakeVector4(
      X=1
      Y=1
      Z=1
      W=&FloatFromCurve(
       FloatCurve=DataInterface() {propertyValues {Curve {
          defaultValue=0
          keys [{interpMode=RCIM_Cubic, tangentMode=RCTM_User, time=0.2, value=1}, {arriveTangent=-7.5916, tangentMode=RCTM_User, time=0.4, value=0}]
         }}}
       CurveIndex=$System.PortalOpen
      )
     )
    )) {isSetParameters=true}
   &ScaleSpriteSize(
    Uniform Curve Index=$System.PortalOpen
    Uniform Curve Sprite Scale=DataInterface() {propertyValues {Curve {
       defaultValue=0
       keys [
        {arriveTangent=24.67, interpMode=RCIM_Cubic, time=0.05, value=0}
        {interpMode=RCIM_Cubic, tangentMode=RCTM_User, time=0.1, value=1}
        {arriveTangent=-7.6803, tangentMode=RCTM_User, time=0.5, value=0}
       ]
      }}}
   )
   &ScaleMeshSize(Scale Factor=&VectorFromFloat(Value=&FloatFromCurve(
      FloatCurve=DataInterface() {propertyValues {Curve {defaultValue=1, keys [{arriveTangent=6.0675, interpMode=RCIM_Cubic, time=0, value=0}, {tangentMode=RCTM_User, time=0.2, value=1}]}}}
      CurveIndex=$System.PortalOpen
     )))
   &DynamicMaterialParameters(
    Index 0 Param 1=$System.PortalOpen
    Index 1 Param 1=&Lerp_Float(A=500, B=1400, Alpha=$System.PortalOpen)
    Index 1 Param 2=20
    Index 3 Param 1=1
    Index 3 Param 2=1
    Index 3 Param 3=1
    Module.Index0Float0Write=true
    Module.Index0Float1Write=false
    Module.Index0Float2Write=false
    Module.Index0Float3Write=false
    Module.Index1Float0Write=true
    Module.Index1Float1Write=true
    Module.Index1Float2Write=true
    Module.Index1Float3Write=false
    Module.Index3Float0Write=true
    Module.Index3Float1Write=true
    Module.Index3Float2Write=true
    Module.Index3Float3Write=true
    Write Parameter Index 1=::NewEnumerator1
    Write Parameter Index 3=::NewEnumerator1
   )
   /SetVariables(Particles.RingScale:Vector3f=&Multiply_VectorByFloat(Vector=$Particles.Initial.Scale, Float=&Add_Float(A=$System.PortalOpen, B=0.04))) {isSetParameters=true}
  ]
  rendererClasses [&NiagaraMeshRendererProperties, &NiagaraSpriteRendererProperties]
  renderers [
   {rendererClass=&NiagaraMeshRendererProperties, rendererIndex=0}
   {rendererClass=&NiagaraMeshRendererProperties, rendererIndex=1}
   {rendererClass=&NiagaraSpriteRendererProperties, rendererIndex=2}
  ]
 }
 ShardsRadial {
  emitterUpdateScript [
   &EmitterState(Life Cycle Mode=::NewEnumerator1, Loop Behavior=::NewEnumerator1, Loop Duration=0.7, UseLoopDelay=false)
   &SpawnRate(
    SpawnRate=&FloatFromCurve(
     FloatCurve=DataInterface() {propertyValues {Curve {keys [{arriveTangent=-2.7768, interpMode=RCIM_Cubic, time=0, value=1}, <at 0.75 0>]}}}
     CurveIndex=$Emitter.Age
     Scale Curve=300
    )
    Use Spawn Probability=false
   )
   &SpawnBurst_Instantaneous(Spawn Count=25, Spawn Time=0.05, Use Spawn Probability=false, UseLoopCountLimit=false)
  ]
  particleSpawnScript [
   &InitializeParticle(
    Color=&MakeCustomLinearColorFromBool(
     Bool=$User.RedRift
     True Color=&RandomRangeLinearColor(Minimum=<rgba 2 0.0584 0.1348 1>, Maximum=<rgba 5 0.2445 0.149 1>)
     False Color=&RandomRangeLinearColor(Minimum=<rgba 2 0.2948 0.0651 1>, Maximum=<rgba 5 1.146 0.1353 1>)
    )
    Lifetime Max=5
    Lifetime Min=3
    Lifetime Mode=::NewEnumerator1
    Mesh Scale Max=<30 30 30>
    Mesh Scale Min=<4 4 5>
    Mesh Scale Mode=::NewEnumerator2
    Sprite Size Mode=::NewEnumerator1
    Uniform Sprite Size Max=2500
    Uniform Sprite Size Min=1250
    UsePositionOffset=false
    Write Material Random=true
   )
   /SetVariables(Particles.MeshIndex=&RandomRangeInt(Minimum=0, Maximum=&RandomRangeInt(Minimum=0, Maximum=1))) {isSetParameters=true}
   &ShapeLocation(
    Cylinder Height=50
    Cylinder Radius=&Lerp_Float(A=100, B=5000, Alpha=$NPC.NPC_DragonCart_Orb_Rift.PortalOpen)
    MaskSpawnGroup=false
    Override Seed=false
    Shape Primitive=::NewEnumerator1
    Surface Only=true
    Use Endcaps In Surface Only Mode=false
    UseOwnerScale=true
   )
   &AddVelocity(
    Module.Use Falloff Distance=false
    Origin Offset=&MaskVector(Mask Source=0, Mask Parameter=?, Mask=1, Masked Vector=<-1000 0 0>, Passthrough Vector=<0 0 0>)
    Override Seed=false
    Velocity Mode=::NewEnumerator1
    Velocity Speed=&Multiply_Float(
     A=&RandomRangeFloat(Minimum=3000, Maximum=18000)
     B=&MaskFloatBySpawnGroup(Masked Float=0.6, Passthrough Float=1, Spawn Group Mask=1)
    )
   )
   &InitialMeshOrientation(Use Rotation Vector=false)
  ]
  particleUpdateScript [
   &ParticleState()
   &ScaleSpriteSize(
    Uniform Curve Index=&Add_Float(
     A=&Multiply_Float(A=$Particles.Age, B=&RandomRangeFloat(Minimum=0.8, Maximum=1.5))
     B=&RandomRangeFloat(Minimum=-0.25, Maximum=0.25)
    )
    Uniform Curve Sprite Scale=DataInterface() {propertyValues {Curve {keys [
        {tangentMode=RCTM_User, time=0.2, value=0}
        <at 0.22 1 RCIM_Cubic>
        {tangentMode=RCTM_User, time=0.5, value=0}
        <at 0.68 0>
        <at 0.7 0.8 RCIM_Cubic>
        <at 0.85 0>
       ]}}}
   )
   &UpdateMeshOrientation(
    Rotation Rate=&FloatFromCurve(
     FloatCurve=DataInterface() {propertyValues {Curve {keys [
         <at 0 1 RCIM_Cubic>
         {arriveTangent=-2.1341, interpMode=RCIM_Cubic, leaveTangent=-2.1341, tangentMode=RCTM_User, time=0.35, value=0.65}
         <at 0.8 0.1 RCIM_Cubic>
        ]}}}
     Scale Curve=&RandomRangeFloat(Minimum=0.5, Maximum=1)
    )
    Rotation Vector=&RandomVector(Vector Scale=1)
   )
   &Drag(
    Drag=&FloatFromCurve(
     FloatCurve=DataInterface() {propertyValues {Curve {keys [
         {arriveTangent=4.7895, interpMode=RCIM_Cubic, leaveTangent=4.7895, tangentMode=RCTM_User, time=0, value=0.1}
         {arriveTangent=0.6, interpMode=RCIM_Cubic, leaveTangent=0.6, time=0.25, value=1}
         {arriveTangent=-4.8446, interpMode=RCIM_Cubic, leaveTangent=-4.8446, tangentMode=RCTM_User, time=0.5, value=0.4}
        ]}}}
     Scale Curve=&RandomRangeFloat(Minimum=1, Maximum=2.5)
    )
    Use Linear Drag=true
    Use Rotational Drag=false
   )
   &SolveForcesAndVelocity(Clamp Velocity=false, Limit Acceleration=false)
   &ScaleColor(
    Scale Alpha=&FloatFromCurve(FloatCurve=DataInterface() {propertyValues {Curve {keys [<at 0.5 1 RCIM_Cubic>,<at 1 0>]}}})
    Scale RGB=&VectorFromFloat(Value=&FloatFromCurve(FloatCurve=DataInterface() {propertyValues {Curve {keys [{arriveTangent=-13.9762, interpMode=RCIM_Cubic, leaveTangent=-13.9762, tangentMode=RCTM_User, time=0, value=5}, <at 0.5 1>]}}}))
    ScaleA=true
    ScaleRGB=true
   )
   /SetVariables(Particles.FlareColor=&RandomRangeLinearColor(Minimum=<rgba 5 1.8671 0.1069 1>, Maximum=<rgba 10 6.6713 0.4732 1>)) {isSetParameters=true}
  ]
  rendererClasses [&NiagaraMeshRendererProperties, &NiagaraSpriteRendererProperties]
  renderers [
   {rendererClass=&NiagaraMeshRendererProperties, rendererIndex=0}
   {rendererClass=&NiagaraSpriteRendererProperties, rendererIndex=1}
  ]
 }
 SparklesRadial {
  simTarget=GPUComputeSim
  emitterUpdateScript [
   &EmitterState(Life Cycle Mode=::NewEnumerator1, Loop Behavior=::NewEnumerator1, Loop Duration Mode=::NewEnumerator1, UseLoopDelay=false)
   &SpawnRate(
    SpawnRate=&Add_Float(
     A=&FloatFromCurve(
      FloatCurve=DataInterface() {propertyValues {Curve {
         defaultValue=1
         keys [
          {arriveTangent=6.3944, interpMode=RCIM_Cubic, leaveTangent=6.3944, tangentMode=RCTM_User, time=0, value=0}
          {interpMode=RCIM_Cubic, tangentMode=RCTM_User, time=0.5, value=1}
          {tangentMode=RCTM_User, time=1, value=0.3}
         ]
        }}}
      CurveIndex=$NPC.NPC_DragonCart_Orb_Rift.PortalOpen
      Scale Curve=500
     )
     B=&FloatFromCurve(
      FloatCurve=DataInterface() {propertyValues {Curve {keys [
          {arriveTangent=-1.8943, interpMode=RCIM_Cubic, time=0, value=1}
          {arriveTangent=-1, interpMode=RCIM_Cubic, leaveTangent=-1, time=0.15, value=0.5}
          <at 1 0>
         ]}}}
      CurveIndex=$Emitter.Age
      Scale Curve=800
     )
    )
    Use Spawn Probability=false
   )
   &SpawnBurst_Instantaneous(Spawn Count=100, Use Spawn Probability=false, UseLoopCountLimit=false)
  ]
  particleSpawnScript [
   &InitializeParticle(
    Color=&MakeCustomLinearColorFromBool(
     Bool=$User.RedRift
     True Color=&RandomRangeLinearColor(Minimum=<rgba 3 0.0844 0.2564 1>, Maximum=<rgba 7 0.2486 0 1>)
     False Color=&RandomRangeLinearColor(Minimum=<rgba 3 0.6414 0.0936 1>, Maximum=<rgba 7 3.5851 0.2277 1>)
    )
    Lifetime=5
    Sprite Size Mode=::NewEnumerator1
    Uniform Sprite Size Max=150
    Uniform Sprite Size Min=50
    UsePositionOffset=false
    Write Material Random=true
   )
   &ShapeLocation(
    Cylinder Height=500
    Cylinder Radius=&Lerp_Float(A=100, B=5200, Alpha=$NPC.NPC_DragonCart_Orb_Rift.PortalOpen)
    MaskSpawnGroup=false
    Non Uniform Scale=<1 1 0.5>
    Override Seed=false
    Shape Primitive=::NewEnumerator1
    Surface Only=true
    Use Endcaps In Surface Only Mode=false
    UseOwnerScale=true
   )
   &AddVelocity(
    Module.Use Falloff Distance=false
    Origin Offset=&MaskVector(Mask Source=0, Mask Parameter=?, Mask=1, Masked Vector=<-2000 0 0>, Passthrough Vector=<0 0 0>)
    Override Seed=false
    Velocity Mode=::NewEnumerator1
    Velocity Speed=&Lerp_Float(
     A=2500
     B=18000
     Alpha=&FloatFromCurve(
      FloatCurve=DataInterface() {propertyValues {Curve {
         defaultValue=1
         keys [
          {arriveTangent=24.2823, interpMode=RCIM_Cubic, leaveTangent=24.2823, tangentMode=RCTM_User, time=0, value=0}
          {arriveTangent=-1.9963, interpMode=RCIM_Cubic, leaveTangent=-1.9963, tangentMode=RCTM_User, time=0.1, value=1}
          {tangentMode=RCTM_User, time=1, value=0}
         ]
        }}}
      CurveIndex=$Emitter.Age
      Scale Curve=&RandomRangeFloat(Minimum=0.7, Maximum=1)
     )
    )
   )
   /SetVariables(Particles.Lifetime=&Multiply_Float(A=&VectorLength(Vector=$Particles.Velocity), B=0.001)) {isSetParameters=true}
  ]
  particleUpdateScript [
   &ParticleState()
   &DynamicMaterialParameters(
    Index 0 Param 1=&FloatFromCurve(
     FloatCurve=DataInterface() {propertyValues {Curve {keys [{arriveTangent=-1.4828, interpMode=RCIM_Cubic, leaveTangent=-1.4828, tangentMode=RCTM_User, time=0, value=1}, <at 1 0>]}}}
     Scale Curve=&RandomRangeFloat(Minimum=0, Maximum=1)
    )
    Index 0 Param 2=&RandomRangeFloat(Minimum=0.5, Maximum=0.8)
    Module.Index0Param0Write=true
    Module.Index0Param1Write=true
    Module.Index0Param2Write=false
    Module.Index0Param3Write=false
   )
   &ScaleSpriteSize(
    Uniform Curve Scale=&FloatFromCurve(FloatCurve=DataInterface() {propertyValues {Curve {keys [<at 5000 0.5>,<at 20000 1>]}}}, CurveIndex=$Engine.Owner.LODDistance)
    Uniform Curve Sprite Scale=DataInterface() {propertyValues {Curve {keys [
        {arriveTangent=68.9421, interpMode=RCIM_Cubic, leaveTangent=68.9421, tangentMode=RCTM_User, time=0, value=0}
        {interpMode=RCIM_Cubic, tangentMode=RCTM_User, time=0.05, value=1.25}
        {arriveTangentWeight=0.0498, interpMode=RCIM_Cubic, leaveTangentWeight=0.3986, tangentMode=RCTM_User, time=0.15, value=1}
        {arriveTangent=-3.1615, tangentMode=RCTM_User, time=1, value=0}
       ]}}}
   )
   &Drag(
    Drag=&FloatFromCurve(
     FloatCurve=DataInterface() {propertyValues {Curve {keys [
         {arriveTangent=4.7895, interpMode=RCIM_Cubic, leaveTangent=4.7895, tangentMode=RCTM_User, time=0, value=0.1}
         {arriveTangent=0.3, interpMode=RCIM_Cubic, leaveTangent=0.3, time=0.25, value=1}
         {interpMode=RCIM_Cubic, leaveTangent=-4.8446, time=0.5, value=0.25}
        ]}}}
     Scale Curve=&RandomRangeFloat(Minimum=1, Maximum=2.5)
    )
    Use Linear Drag=true
    Use Rotational Drag=false
   )
   &VectorNoiseForce(Force Amount=2000, Module.Use Random Noise Time Interval=false)
   &SolveForcesAndVelocity(Clamp Velocity=false, Limit Acceleration=false)
   &ScaleColor(
    Scale Alpha=&FloatFromCurve(FloatCurve=DataInterface() {propertyValues {Curve {keys [<at 0.5 1 RCIM_Cubic>, {arriveTangent=-4.8071, tangentMode=RCTM_User, time=1, value=0}]}}})
    Scale RGB=&VectorFromFloat(Value=&FloatFromCurve(FloatCurve=DataInterface() {propertyValues {Curve {keys [{arriveTangent=-13.9762, interpMode=RCIM_Cubic, leaveTangent=-13.9762, tangentMode=RCTM_User, time=0, value=5}, <at 0.5 1>]}}}))
    ScaleA=true
    ScaleRGB=true
   )
  ]
  rendererClasses [&NiagaraSpriteRendererProperties]
  renderers [{rendererClass=&NiagaraSpriteRendererProperties}]
 }
 WaveMesh {
  emitterUpdateScript [
   &EmitterState(Life Cycle Mode=::NewEnumerator1, Loop Behavior=::NewEnumerator1, Loop Duration=1, UseLoopDelay=false)
   &SpawnBurst_Instantaneous(Spawn Count=5, Use Spawn Probability=false, UseLoopCountLimit=false)
  ]
  particleSpawnScript [
   &InitializeParticle(
    Color=&MakeCustomLinearColorFromBool(Bool=$User.RedRift, True Color=<rgba 20 4.8854 0.4511 1>, False Color=<rgba 20 9.0988 0.5907 1>)
    Lifetime=4
    Mesh Scale=&Multiply_VectorByFloat(Vector=<1 1 1>, Float=100)
    Mesh Scale Mode=::NewEnumerator4
    Sprite Size Mode=::NewEnumerator3
    Uniform Sprite Size=38000
    UsePositionOffset=false
    Write Material Random=true
   )
   &CameraOffset(Camera Offset Amount=500)
   &InitialMeshOrientation(Mesh Orientation Mode=::NewEnumerator2, Use Rotation Vector=false)
   &SpriteFacingAndAlignment(Sprite Facing=<0 0 1>, WriteAlignment=false, WriteFacing=true)
  ]
  particleUpdateScript [
   &ParticleState()
   &ScaleSpriteSize(Uniform Curve Sprite Scale=DataInterface() {propertyValues {Curve {keys [
        {arriveTangent=2.1385, interpMode=RCIM_Cubic, time=0, value=0}
        {arriveTangent=3.8333, interpMode=RCIM_Cubic, leaveTangent=3.8333, tangentMode=RCTM_User, time=0.14, value=0.45}
        {arriveTangent=0.2576, interpMode=RCIM_Cubic, leaveTangent=0.2576, tangentMode=RCTM_User, time=0.5, value=1}
        {interpMode=RCIM_Cubic, leaveTangent=0.9937, time=0.7, value=0}
       ]}}})
   &ScaleMeshSize(Scale Factor=&VectorFromFloat(Value=&FloatFromCurve(FloatCurve=DataInterface() {propertyValues {Curve {keys [
          {arriveTangent=3.7995, interpMode=RCIM_Cubic, time=0, value=0}
          {arriveTangent=1.2327, interpMode=RCIM_Cubic, leaveTangent=1.2327, tangentMode=RCTM_User, time=0.1968, value=0.5}
          {arriveTangent=0.2576, interpMode=RCIM_Cubic, leaveTangent=0.2576, tangentMode=RCTM_User, time=1, value=1}
         ]}}})))
   &DynamicMaterialParameters(
    Index 0 Param 1=&FloatFromCurve(
     FloatCurve=DataInterface() {propertyValues {Curve {keys [
         {arriveTangent=-2.1702, interpMode=RCIM_Cubic, time=0, value=1}
         {arriveTangent=-4.0704, interpMode=RCIM_Cubic, leaveTangent=-4.0704, tangentMode=RCTM_User, time=0.3, value=0.625}
         {arriveTangent=-2.2806, interpMode=RCIM_Cubic, leaveTangent=-2.2806, tangentMode=RCTM_User, time=1, value=-1}
        ]}}}
     Scale Curve=3
    )
    Index 0 Param 2=-0.25
    Index 0 Param 3=2
    Index 3 Param 1=1
    Index 3 Param 2=0.2
    Module.Param0WriteEnabled=true
    Module.Param0WriteEnabled3=true
    Module.Param1WriteEnabled=true
    Module.Param1WriteEnabled3=true
    Module.Param2WriteEnabled=true
    Module.Param2WriteEnabled3=false
    Module.Param3WriteEnabled=false
    Module.Param3WriteEnabled3=false
    Write Parameter Index 3=true
   )
   &ScaleColor(
    Scale Alpha=&FloatFromCurve(FloatCurve=DataInterface() {propertyValues {Curve {keys [
         {arriveTangent=-2.2763, interpMode=RCIM_Cubic, time=0, value=1}
         {arriveTangent=-2.3237, interpMode=RCIM_Cubic, leaveTangent=-2.3237, tangentMode=RCTM_User, time=0.2119, value=0.3471}
         {arriveTangent=-0.2121, leaveTangent=-0.3161, tangentMode=RCTM_User, time=0.4, value=0.1}
         {arriveTangent=-0.1021, tangentMode=RCTM_User, time=1, value=0}
        ]}}})
    Scale RGB=&VectorFromFloat(Value=&FloatFromCurve(FloatCurve=DataInterface() {propertyValues {Curve {keys [
          {arriveTangent=-2.3948, interpMode=RCIM_Cubic, time=0, value=1}
          {arriveTangent=-0.5342, interpMode=RCIM_Cubic, leaveTangent=-0.5342, tangentMode=RCTM_User, time=0.15, value=0.35}
          <at 1 0.2>
         ]}}}))
    ScaleA=true
    ScaleRGB=true
   )
   /SetVariables(Particles.SpriteColor=&MultiplyLinearColor(
     Linear Color=<rgba 2 0.6798 0.0938 1>
     Scale Factor=&MakeVector4FromVectorAndFloat(
      "Vector (RGB)"=<1 1 1>
      "Float (Alpha)"=&FloatFromCurve(FloatCurve=DataInterface() {propertyValues {Curve {keys [
           {arriveTangent=-2.2763, interpMode=RCIM_Cubic, time=0, value=1}
           {arriveTangent=-1.6136, interpMode=RCIM_Cubic, leaveTangent=-1.6136, tangentMode=RCTM_User, time=0.2578, value=0.2176}
           {arriveTangent=-0.1021, tangentMode=RCTM_User, time=0.6, value=0}
          ]}}})
     )
    )) {isSetParameters=true}
  ]
  rendererClasses [&NiagaraMeshRendererProperties, &NiagaraSpriteRendererProperties]
  renderers [
   {rendererClass=&NiagaraMeshRendererProperties, rendererIndex=0}
   {rendererClass=&NiagaraSpriteRendererProperties, rendererIndex=1}
  ]
 }
}
summary {
 systemName=NS_DragonCart_Animated_Portal
 emitters [
  {emitterName=PortalMesh, rendererClasses [&NiagaraMeshRendererProperties, &NiagaraSpriteRendererProperties]}
  {emitterName=Cracks, rendererClasses [&NiagaraMeshRendererProperties]}
  {emitterName=WaveMesh, rendererClasses [&NiagaraMeshRendererProperties, &NiagaraSpriteRendererProperties]}
  {emitterName=ShardsRadial, rendererClasses [&NiagaraMeshRendererProperties, &NiagaraSpriteRendererProperties]}
  {emitterName=SparklesRadial, simTarget=GPUComputeSim, rendererClasses [&NiagaraSpriteRendererProperties]}
  {emitterName=Lights_High, rendererClasses [&NiagaraSpriteRendererProperties, &NiagaraLightRendererProperties]}
 ]
 userVariables [User.RedRift {type=NiagaraBool, defaultValue=true}]
}
