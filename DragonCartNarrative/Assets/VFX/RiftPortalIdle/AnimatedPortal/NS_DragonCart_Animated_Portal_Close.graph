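# Niagara effect for the DragonCart animated rift portal closing: a warm energy flash with a spray of sparks.
# Energy_Dark_Glow bursts a mesh, large sprite and light, animated by size, alpha and light-color curves.
# GPU SparklesRadial combines a burst with decaying spawn rate on a cylinder whose radius follows the PortalOpen parameter-collection value.
# Sparks get an early velocity spike, drag and noise, with size scaled by LOD distance.
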NiagaraSystem /DragonCartNarrative/Assets/VFX/RiftPortalIdle/AnimatedPortal/NS_DragonCart_Animated_Portal_Close
dependencies {
 usedDataInterfaces [&NiagaraDataInterfaceCurve]
 usedDynamicInputs [&Subtract_Float]
 usedModules [&SystemState]
}
emitters {
 Energy_Dark_Glow {
  emitterUpdateScript [
   &EmitterState()
   &SpawnBurst_Instantaneous(Use Spawn Probability=false, UseLoopCountLimit=false)
  ]
  particleSpawnScript [
   &InitializeParticle(
    Color=<rgba 5 1.6177 0 1>
    Lifetime=1.5
    Mesh Scale Mode=::NewEnumerator3
    Mesh Uniform Scale=250
    Sprite Size=<30000 12000>
    Sprite Size Mode=::NewEnumerator4
    UsePositionOffset=false
    Write Material Random=true
   )
   &InitialMeshOrientation(Mesh Axis To Orient=<0 0 -1>, Mesh Orientation Mode=::NewEnumerator2, Use Rotation Vector=true)
   /SetVariables(Particles.FlareDynamic:Vector4f=&MakeVector4(X=1, Y=1, Z=0, W=0)) {isSetParameters=true}
   /SetVariables(Particles.DynamicPlasmaEnergy:Vector4f=&MakeVector4(X=0, Y=0, Z=-0.25, W=0)) {isSetParameters=true}
   &CameraOffset(Camera Offset Amount=1000)
  ]
  particleUpdateScript [
   &ParticleState()
   &ScaleMeshSize(Scale Factor=&VectorFromFloat(Value=&FloatFromCurve(FloatCurve=DataInterface() {propertyValues {Curve {
         defaultValue=1
         keys [
          {arriveTangent=2.4257, interpMode=RCIM_Cubic, leaveTangent=2.4257, tangentMode=RCTM_User, time=0, value=0}
          {arriveTangent=0.3328, tangentMode=RCTM_User, time=1, value=1}
         ]
        }}})))
   &ScaleSpriteSize(Uniform Curve Sprite Scale=DataInterface() {propertyValues {Curve {keys [
        {arriveTangent=4.1958, interpMode=RCIM_Cubic, leaveTangent=4.1958, tangentMode=RCTM_User, time=0, value=0.2}
        {arriveTangentWeight=0.1699, interpMode=RCIM_Cubic, leaveTangentWeight=0.1734, tangentMode=RCTM_User, time=0.5, value=1}
        {interpMode=RCIM_Cubic, leaveTangent=-4.3111, time=1, value=0}
       ]}}})
   &ScaleColor(
    Scale Alpha=&FloatFromCurve(FloatCurve=DataInterface() {propertyValues {Curve {keys [{arriveTangent=-1.7165, interpMode=RCIM_Cubic, leaveTangent=-1.7165, tangentMode=RCTM_User, time=0, value=1}, <at 1 0>]}}})
    ScaleA=true
    ScaleRGB=true
   )
   /SetVariables(Particles.LightColor=&Multiply_LinearColorByFloat(
     Linear Color=$Particles.Color
     Float=&FloatFromCurve(
      FloatCurve=DataInterface() {propertyValues {Curve {
         defaultValue=1
         keys [
          {arriveTangent=7.0556, interpMode=RCIM_Cubic, leaveTangent=7.0556, tangentMode=RCTM_User, time=0, value=0}
          {tangentMode=RCTM_User, time=0.3, value=1}
         ]
        }}}
      Scale Curve=10
     )
    )) {isSetParameters=true}
  ]
  rendererClasses [
   &NiagaraMeshRendererProperties
   &NiagaraSpriteRendererProperties
   &NiagaraLightRendererProperties
  ]
  renderers [
   {rendererClass=&NiagaraMeshRendererProperties, rendererIndex=0}
   {rendererClass=&NiagaraSpriteRendererProperties, rendererIndex=1}
   {rendererClass=&NiagaraLightRendererProperties, rendererIndex=2}
  ]
 }
 SparklesRadial {
  simTarget=GPUComputeSim
  emitterUpdateScript [
   &EmitterState(Life Cycle Mode=::NewEnumerator1, Loop Behavior=::NewEnumerator1, Loop Duration=1, UseLoopDelay=false)
   &SpawnRate(
    SpawnRate=&FloatFromCurve(
     FloatCurve=DataInterface() {propertyValues {Curve {keys [{arriveTangent=-1.8943, interpMode=RCIM_Cubic, time=0, value=1}, <at 1 0>]}}}
     CurveIndex=$Emitter.Age
     Scale Curve=400
    )
    Use Spawn Probability=false
   )
   &SpawnBurst_Instantaneous(Spawn Count=50, Use Spawn Probability=false, UseLoopCountLimit=false)
  ]
  particleSpawnScript [
   &InitializeParticle(
    Color Maximum=<rgba 7 3.5851 0.2277 1>
    Color Minimum=<rgba 3 0.6414 0.0936 1>
    Color Mode=::NewEnumerator2
    Lifetime Max=3
    Lifetime Min=1.5
    Lifetime Mode=::NewEnumerator1
    Sprite Size Mode=::NewEnumerator1
    Uniform Sprite Size Max=150
    Uniform Sprite Size Min=50
    UsePositionOffset=false
    Write Material Random=true
   )
   &ShapeLocation(
    Cylinder Height=500
    Cylinder Radius=&Lerp_Float(A=100, B=5200, Alpha=$NPC.NPC_DragonCart_Orb_Rift.PortalOpen)
    MaskSpawnGroup=false
    Non Uniform Scale=<1 1 0.5>
    Override Seed=false
    Shape Primitive=::NewEnumerator1
    Surface Only=true
    Use Endcaps In Surface Only Mode=false
    UseOwnerScale=true
   )
   &AddVelocity(
    Module.Use Falloff Distance=false
    Origin Offset=&MaskVector(Mask Source=0, Mask Parameter=?, Mask=1, Masked Vector=<-2000 0 0>, Passthrough Vector=<0 0 0>)
    Override Seed=false
    Velocity Mode=::NewEnumerator1
    Velocity Speed=&Lerp_Float(
     A=1500
     B=15000
     Alpha=&FloatFromCurve(
      FloatCurve=DataInterface() {propertyValues {Curve {
         defaultValue=1
         keys [
          {arriveTangent=24.2823, interpMode=RCIM_Cubic, leaveTangent=24.2823, tangentMode=RCTM_User, time=0, value=0}
          {arriveTangent=-5.5233, interpMode=RCIM_Cubic, leaveTangent=-5.5233, tangentMode=RCTM_User, time=0.1, value=1}
          {tangentMode=RCTM_User, time=0.5, value=0}
         ]
        }}}
      CurveIndex=$Emitter.Age
      Scale Curve=&RandomRangeFloat(Minimum=0.7, Maximum=1)
     )
    )
   )
   &CameraOffset(Camera Offset Amount=500)
  ]
  particleUpdateScript [
   &ParticleState()
   &DynamicMaterialParameters(
    Index 0 Param 1=&FloatFromCurve(
     FloatCurve=DataInterface() {propertyValues {Curve {keys [{arriveTangent=-1.4828, interpMode=RCIM_Cubic, leaveTangent=-1.4828, tangentMode=RCTM_User, time=0, value=1}, <at 1 0>]}}}
     Scale Curve=&RandomRangeFloat(Minimum=0, Maximum=1)
    )
    Index 0 Param 2=&RandomRangeFloat(Minimum=0.5, Maximum=0.8)
    Module.Index0Param0Write=true
    Module.Index0Param1Write=true
    Module.Index0Param2Write=false
    Module.Index0Param3Write=false
   )
   &ScaleSpriteSize(
    Uniform Curve Scale=&FloatFromCurve(FloatCurve=DataInterface() {propertyValues {Curve {keys [<at 5000 0.5>,<at 20000 1>]}}}, CurveIndex=$Engine.Owner.LODDistance)
    Uniform Curve Sprite Scale=DataInterface() {propertyValues {Curve {keys [
        {arriveTangent=24.5678, interpMode=RCIM_Cubic, leaveTangent=24.5678, tangentMode=RCTM_User, time=0, value=0}
        {arriveTangentWeight=0.0498, interpMode=RCIM_Cubic, leaveTangentWeight=0.3986, tangentMode=RCTM_User, time=0.1, value=1}
        {arriveTangent=-3.1615, tangentMode=RCTM_User, time=1, value=0}
       ]}}}
   )
   &Drag(
    Drag=&FloatFromCurve(
     FloatCurve=DataInterface() {propertyValues {Curve {keys [
         {arriveTangent=4.7895, interpMode=RCIM_Cubic, leaveTangent=4.7895, tangentMode=RCTM_User, time=0, value=0.1}
         {arriveTangent=0.3, interpMode=RCIM_Cubic, leaveTangent=0.3, time=0.25, value=1}
         {interpMode=RCIM_Cubic, leaveTangent=-4.8446, time=0.5, value=0.25}
        ]}}}
     Scale Curve=&RandomRangeFloat(Minimum=1.5, Maximum=3)
    )
    Use Linear Drag=true
    Use Rotational Drag=false
   )
   &VectorNoiseForce(Force Amount=2000, Module.Use Random Noise Time Interval=false)
   &SolveForcesAndVelocity(Clamp Velocity=false, Limit Acceleration=false)
   &ScaleColor(
    Scale Alpha=&FloatFromCurve(FloatCurve=DataInterface() {propertyValues {Curve {keys [<at 0.5 1 RCIM_Cubic>, {arriveTangent=-4.8071, tangentMode=RCTM_User, time=1, value=0}]}}})
    Scale RGB=&VectorFromFloat(Value=&FloatFromCurve(FloatCurve=DataInterface() {propertyValues {Curve {keys [{arriveTangent=-13.9762, interpMode=RCIM_Cubic, leaveTangent=-13.9762, tangentMode=RCTM_User, time=0, value=5}, <at 0.5 1>]}}}))
    ScaleA=true
    ScaleRGB=true
   )
  ]
  rendererClasses [&NiagaraSpriteRendererProperties]
  renderers [{rendererClass=&NiagaraSpriteRendererProperties}]
 }
}
summary {
 systemName=NS_DragonCart_Animated_Portal_Close
 emitters [
  {
   emitterName=Energy_Dark_Glow
   rendererClasses [
    &NiagaraMeshRendererProperties
    &NiagaraSpriteRendererProperties
    &NiagaraLightRendererProperties
   ]
  }
  {emitterName=SparklesRadial, simTarget=GPUComputeSim, rendererClasses [&NiagaraSpriteRendererProperties]}
 ]
 userVariables [User.Lifetime {type=NiagaraFloat, defaultValue=20, description=""}]
}
