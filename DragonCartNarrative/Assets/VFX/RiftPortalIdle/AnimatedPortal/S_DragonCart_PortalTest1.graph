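# Level Sequence, apparently a test, animating a rift portal in DragonCartNarrative that builds up, opens as an orb actor descends, then closes.
# Oscillating, rising Niagara scale and intensity parameters on the rift actor accompany front and back buildup effects, which hand over to spawn-tracked portal and closing effects.
# A Material Parameter Collection track drives portal openness and position, while the orb gets masked material swaps, descends and moves sideways.
# A subsequence track spans the timeline.
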LevelSequence /DragonCartNarrative/Assets/VFX/RiftPortalIdle/AnimatedPortal/S_DragonCart_PortalTest1
bindings [
 NS_DragonCart_Animated_Portal_Close {
  id=#1
  possessed_class=NiagaraActor
  tracks [
   {
    type=MovieSceneSpawnTrack
    sections [{type=MovieSceneSpawnSection, channels [None {type=MovieSceneScriptingBoolChannel, default=false, keys [[584.0025,false],[714,true]]}]}]
   }
   {
    type=MovieScene3DTransformTrack
    sections [{
      type=MovieScene3DTransformSection
      channels [
       Location.X {type=MovieSceneScriptingDoubleChannel, default=-10752.0001}
       Location.Y {type=MovieSceneScriptingDoubleChannel, default=-5632}
       Location.Z {type=MovieSceneScriptingDoubleChannel, default=17700}
       Rotation.X {type=MovieSceneScriptingDoubleChannel, default=0}
       Rotation.Y {type=MovieSceneScriptingDoubleChannel, default=0}
       Rotation.Z {type=MovieSceneScriptingDoubleChannel, default=0}
       Scale.X {type=MovieSceneScriptingDoubleChannel, default=1}
       Scale.Y {type=MovieSceneScriptingDoubleChannel, default=1}
       Scale.Z {type=MovieSceneScriptingDoubleChannel, default=1}
      ]
     }]
   }
  ]
 }
 BP_DCN_Rift_Actor {
  id=#2
  possessed_class=BP_DCN_Rift_Actor_C
  tracks [{
    type=MovieSceneVisibilityTrack
    sections [{type=MovieSceneVisibilitySection, channels [None {type=MovieSceneScriptingBoolChannel, default=true, const [true,1]}]}]
   }]
 }
 SM_DragonCart_Portal {
  id=#3
  possessed_class=NiagaraActor
  tracks [
   {
    type=MovieSceneSpawnTrack
    sections [{
      type=MovieSceneSpawnSection
      channels [None {type=MovieSceneScriptingBoolChannel, default=false, keys [[0,false],[30,true],[830.835,false]]}]
     }]
   }
   {
    type=MovieScene3DTransformTrack
    sections [{
      type=MovieScene3DTransformSection
      channels [
       Location.X {type=MovieSceneScriptingDoubleChannel, default=-10752.0001}
       Location.Y {type=MovieSceneScriptingDoubleChannel, default=-5632}
       Location.Z {type=MovieSceneScriptingDoubleChannel, default=17813.3003}
       Rotation.X {type=MovieSceneScriptingDoubleChannel, default=0}
       Rotation.Y {type=MovieSceneScriptingDoubleChannel, default=0}
       Rotation.Z {type=MovieSceneScriptingDoubleChannel, default=0}
       Scale.X {type=MovieSceneScriptingDoubleChannel, default=1}
       Scale.Y {type=MovieSceneScriptingDoubleChannel, default=1}
       Scale.Z {type=MovieSceneScriptingDoubleChannel, default=1}
      ]
     }]
   }
  ]
 }
 NS_DCN_Rift_Buildup_Back {
  id=#4
  possessed_class=NiagaraActor
  tracks [
   {
    type=MovieSceneSpawnTrack
    sections [{
      type=MovieSceneSpawnSection
      channels [None {type=MovieSceneScriptingBoolChannel, default=false, keys [[-1795,false],[-1698,true],[30,false]]}]
     }]
   }
   {
    type=MovieScene3DTransformTrack
    sections [{
      type=MovieScene3DTransformSection
      channels [
       Location.X {type=MovieSceneScriptingDoubleChannel, default=-10752.0001}
       Location.Y {type=MovieSceneScriptingDoubleChannel, default=-5632}
       Location.Z {type=MovieSceneScriptingDoubleChannel, default=17813.3003}
       Rotation.X {type=MovieSceneScriptingDoubleChannel, default=0}
       Rotation.Y {type=MovieSceneScriptingDoubleChannel, default=0}
       Rotation.Z {type=MovieSceneScriptingDoubleChannel, default=0}
       Scale.X {type=MovieSceneScriptingDoubleChannel, default=1}
       Scale.Y {type=MovieSceneScriptingDoubleChannel, default=1}
       Scale.Z {type=MovieSceneScriptingDoubleChannel, default=1}
      ]
     }]
   }
  ]
 }
 NiagaraComponent0 {id=#5, parent_id=#6, possessed_class=NiagaraComponent, tracks []}
 NS_DCN_Rift_Buildup_Front {
  id=#6
  possessed_class=NiagaraActor
  tracks [
   {
    type=MovieSceneSpawnTrack
    sections [{
      type=MovieSceneSpawnSection
      channels [None {type=MovieSceneScriptingBoolChannel, default=false, keys [[-1795,false],[-1700,true],[30,false]]}]
     }]
   }
   {
    type=MovieScene3DTransformTrack
    sections [{
      type=MovieScene3DTransformSection
      channels [
       Location.X {type=MovieSceneScriptingDoubleChannel, default=-10752.0001}
       Location.Y {type=MovieSceneScriptingDoubleChannel, default=-5632}
       Location.Z {type=MovieSceneScriptingDoubleChannel, default=17813.3003}
       Rotation.X {type=MovieSceneScriptingDoubleChannel, default=0}
       Rotation.Y {type=MovieSceneScriptingDoubleChannel, default=0}
       Rotation.Z {type=MovieSceneScriptingDoubleChannel, default=0}
       Scale.X {type=MovieSceneScriptingDoubleChannel, default=1}
       Scale.Y {type=MovieSceneScriptingDoubleChannel, default=1}
       Scale.Z {type=MovieSceneScriptingDoubleChannel, default=1}
      ]
     }]
   }
  ]
 }
 OrbHatch {
  id=#7
  parent_id=#8
  possessed_class=StaticMeshComponent
  tracks [{
    type=MovieScenePrimitiveMaterialTrack
    display_name=Material Slot: two
    sections [{
      type=MovieScenePrimitiveMaterialSection
      channels [None {
        type=MovieSceneScriptingObjectPathChannel
        default="<Object '/DragonCartNarrative/Assets/FlyingOrb/Materials/MI_FlyingOrb_A_Masked_Inst' (0x49de44b4800) Class 'MaterialInstanceConstant'>"
       }]
     }]
   }]
 }
 Niagara {
  id=#9
  parent_id=#2
  possessed_class=NiagaraComponent
  tracks [
   {
    type=MovieSceneNiagaraFloatParameterTrack
    display_name=BuildUpIntensity
    sections [{type=MovieSceneFloatSection, channels [None {type=MovieSceneScriptingFloatChannel, default=1.5, keys [[-978,1],[30,2],[594,1.5],[915,1]]}]}]
   }
   {
    type=MovieSceneNiagaraFloatParameterTrack
    display_name=BuildUpScale
    sections [{
      type=MovieSceneFloatSection
      channels [None {
        type=MovieSceneScriptingFloatChannel
        default=1.3
        keys [[-1799,1],[-1726,0.8486],[-1613,0.6008],[-1406,0.8773],[-1269,0.6867],[-1177,0.9359],[-1068,0.7354],[-970,0.9885],[-890,0.7794],[-818,0.921],[-758,0.7093],[-707,1.0262],[-639,0.8004],[-586,0.9541],[-543,0.6831],[-495,1.0673],[-436,0.7865],[-392,0.9509],[-355,0.8374],[-301,1.0796],[-272,0.6895],[-215,1.1177],[-189,0.7951],[-134,1.1625],[-94,0.8724],[-52,1.2196],[-29,0.7069],[-3,1.2559],[15,0.9454],[32,1.3],[50,0.9991]]
       }]
     }]
   }
   {
    type=MovieSceneNiagaraFloatParameterTrack
    display_name=AnimateScale
    sections [{
      type=MovieSceneFloatSection
      channels [None {type=MovieSceneScriptingFloatChannel, default=1.4, keys [[-1795,1],[-1726,0.9],[30,1.4],[594,1.3],[888,1]]}]
     }]
   }
  ]
 }
 StaticMesh {
  id=#10
  parent_id=#8
  possessed_class=StaticMeshComponent
  tracks [
   {
    type=MovieScenePrimitiveMaterialTrack
    display_name=Material Slot: one
    sections [{
      type=MovieScenePrimitiveMaterialSection
      channels [None {
        type=MovieSceneScriptingObjectPathChannel
        default="<Object '/DragonCartNarrative/Assets/FlyingOrb/Materials/M_FlyingOrbB_Masked_Inst' (0x49de44b3800) Class 'MaterialInstanceConstant'>"
        const ["<Object '&M_FlyingOrbB_Masked_Inst' (0x49de44b3800) Class 'MaterialInstanceConstant'>", 1]
       }]
     }]
   }
   {
    type=MovieScenePrimitiveMaterialTrack
    display_name=Material Slot: two
    sections [{
      type=MovieScenePrimitiveMaterialSection
      channels [None {
        type=MovieSceneScriptingObjectPathChannel
        default="<Object '&MI_FlyingOrb_A_Masked_Inst' (0x49de44b4800) Class 'MaterialInstanceConstant'>"
        const ["<Object '&MI_FlyingOrb_A_Masked_Inst' (0x49de44b4800) Class 'MaterialInstanceConstant'>", 1]
       }]
     }]
   }
  ]
 }
 NiagaraComponent0 {id=#11, parent_id=#4, possessed_class=NiagaraComponent, tracks []}
 NiagaraComponent0 {
  id=#12
  parent_id=#3
  possessed_class=NiagaraComponent
  tracks [{
    type=MovieSceneParticleTrack
    display_name=FX System
    sections [{
      type=MovieSceneParticleSection
      channels [None {type=MovieSceneScriptingParticleChannel, keys [[30, "<ParticleKey.ACTIVATE: 0>"], [720, "<ParticleKey.DEACTIVATE: 1>"]]}]
     }]
   }]
 }
 BGA_DCN_OrbActorRotated3 {
  id=#8
  possessed_class=BGA_DCN_OrbActor_C
  tracks [
   {
    type=MovieSceneVisibilityTrack
    sections [{type=MovieSceneVisibilitySection, channels [None {type=MovieSceneScriptingBoolChannel, default=false, keys [[-30,false],[0,true]]}]}]
   }
   {
    type=MovieScene3DTransformTrack
    sections [{
      type=MovieScene3DTransformSection
      channels [
       Location.X {type=MovieSceneScriptingDoubleChannel, default=-10752, const [-10752,2]}
       Location.Y {type=MovieSceneScriptingDoubleChannel, default=-5632, thinned_from=4, keys [[0,-5632],[965.9925,-5632],[1442.9925,0]]}
       Location.Z {
        type=MovieSceneScriptingDoubleChannel
        default=10749.9258
        thinned_from=5
        keys [[0,25973.1504],[28,25973.1506],[930.9925,11136],[1446.9925,11136]]
       }
       Rotation.X {type=MovieSceneScriptingDoubleChannel, default=0, keys [[477,-4.488],[923,0],[965.9925,0],[1071.9925,2.4599],[1442.9925,0]]}
       Rotation.Y {type=MovieSceneScriptingDoubleChannel, default=0, const [0,2]}
       Rotation.Z {type=MovieSceneScriptingDoubleChannel, default=0, const [0,2]}
       Scale.X {type=MovieSceneScriptingDoubleChannel, default=0.75, const [0.75,2]}
       Scale.Y {type=MovieSceneScriptingDoubleChannel, default=0.75, const [0.75,2]}
       Scale.Z {type=MovieSceneScriptingDoubleChannel, default=0.75, const [0.75,2]}
      ]
     }]
   }
  ]
 }
]
display_rate {denominator=1, numerator=60}
tick_resolution {denominator=1, numerator=24000}
tracks [
 {
  type=MovieSceneMaterialParameterCollectionTrack
  display_name=MPC_DragonCart_Orb_Rift
  sections [{
    type=MovieSceneParameterSection
    channels [
     PortalOpen {type=MovieSceneScriptingFloatChannel, default=0.1, keys [[20.8475,0],[30,0.1],[150,1],[570,1],[720,0.1]]}
     PortalPositionR {type=MovieSceneScriptingFloatChannel, const [-10752,1]}
     PortalPositionG {type=MovieSceneScriptingFloatChannel, const [-5632,1]}
     PortalPositionB {type=MovieSceneScriptingFloatChannel, const [17813,1]}
     PortalPositionA {type=MovieSceneScriptingFloatChannel, const [0,1]}
    ]
   }]
 }
 {type=MovieSceneSubTrack, display_name=Subsequences, sections [{type=MovieSceneSubSection, end_frame=1440, start_frame=-1800}]}
]
