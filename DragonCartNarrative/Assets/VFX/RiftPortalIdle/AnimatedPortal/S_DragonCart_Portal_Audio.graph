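# Level Sequence that times the audio for the DragonCart animated rift portal, running about a minute at 30 fps.
# Overlapping audio sections on the orb actor build and release through keyed volume fades, with two late sections also ramping pitch up and back down.
# A Niagara float track raises ZapAudioVolume on the front buildup system during the lead-in.
# The orb also has two event triggers mid-sequence, while the rift actor's audio track is empty.
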LevelSequence /DragonCartNarrative/Assets/VFX/RiftPortalIdle/AnimatedPortal/S_DragonCart_Portal_Audio
bindings [
 NS_DCN_Rift_Buildup_Front {id=#1, possessed_class=NiagaraActor, tracks []}
 NiagaraComponent0 {
  id=#2
  parent_id=#1
  possessed_class=NiagaraComponent
  tracks [{
    type=MovieSceneNiagaraFloatParameterTrack
    display_name=ZapAudioVolume
    sections [{type=MovieSceneFloatSection, channels [None {type=MovieSceneScriptingFloatChannel, default=0.8, keys [[0,0],[749,0.8],[905,1]]}]}]
   }]
 }
 BGA_DCN_OrbActorRotated3 {
  id=#3
  possessed_class=BGA_DCN_OrbActor_C
  tracks [
   {
    type=MovieSceneAudioTrack
    display_name=Audio
    sections [
     {
      type=MovieSceneAudioSection
      end_frame=1081
      start_frame=910
      channels [
       Volume {type=MovieSceneScriptingFloatChannel, default=1}
       Pitch {type=MovieSceneScriptingFloatChannel, default=1}
      ]
     }
     {
      type=MovieSceneAudioSection
      end_frame=1391
      start_frame=1222
      channels [
       Volume {type=MovieSceneScriptingFloatChannel, default=1}
       Pitch {type=MovieSceneScriptingFloatChannel, default=1}
      ]
     }
     {
      type=MovieSceneAudioSection
      end_frame=932
      start_frame=0
      channels [
       Volume {type=MovieSceneScriptingFloatChannel, default=1, keys [[0,0],[689,0.5],[918,1],[931,0],[1190,0]]}
       Pitch {type=MovieSceneScriptingFloatChannel, default=1}
      ]
     }
     {
      type=MovieSceneAudioSection
      end_frame=1234
      start_frame=1191
      channels [
       Volume {type=MovieSceneScriptingFloatChannel, default=0, keys [[1191,0],[1228,1],[1234,0]]}
       Pitch {type=MovieSceneScriptingFloatChannel, default=1}
      ]
     }
     {
      type=MovieSceneAudioSection
      end_frame=926
      start_frame=397
      channels [
       Volume {type=MovieSceneScriptingFloatChannel, default=1}
       Pitch {type=MovieSceneScriptingFloatChannel, default=1}
      ]
     }
     {
      type=MovieSceneAudioSection
      end_frame=922
      start_frame=847
      channels [
       Volume {type=MovieSceneScriptingFloatChannel, default=1}
       Pitch {type=MovieSceneScriptingFloatChannel, default=1}
      ]
     }
     {
      type=MovieSceneAudioSection
      end_frame=1790
      start_frame=979
      channels [
       Volume {type=MovieSceneScriptingFloatChannel, default=2, keys [[1427,1],[1472,2]]}
       Pitch {type=MovieSceneScriptingFloatChannel, default=1}
      ]
     }
     {
      type=MovieSceneAudioSection
      end_frame=1467
      start_frame=1394
      channels [
       Volume {type=MovieSceneScriptingFloatChannel, default=1}
       Pitch {type=MovieSceneScriptingFloatChannel, default=1}
      ]
     }
     {
      type=MovieSceneAudioSection
      end_frame=1360
      start_frame=1002
      channels [
       Volume {type=MovieSceneScriptingFloatChannel, default=0, keys [[1002,0],[1026,1],[1317,1],[1361,0]]}
       Pitch {type=MovieSceneScriptingFloatChannel, default=0.5, keys [[1002,0.5],[1026,1],[1317,1],[1361,0.5]]}
      ]
     }
     {
      type=MovieSceneAudioSection
      end_frame=1610
      start_frame=1392
      channels [
       Volume {type=MovieSceneScriptingFloatChannel, default=0, keys [[1391,0],[1422,1],[1580,1],[1610,0]]}
       Pitch {type=MovieSceneScriptingFloatChannel, default=0.5, keys [[1391,0.5],[1422,1],[1580,1],[1610,0.5]]}
      ]
     }
    ]
   }
   {
    type=MovieSceneEventTrack
    sections [{
      type=MovieSceneEventTriggerSection
      channels [None {
        type=MovieSceneScriptingEventChannel
        keys [[971, "<Struct 'MovieSceneEvent' (0x49ddd17e420) {}>"], [1047, "<Struct 'MovieSceneEvent' (0x49ddd17e380) {}>"]]
       }]
     }]
   }
  ]
 }
 BP_DCN_Rift_Actor {
  id=#4
  possessed_class=BP_DCN_Rift_Actor_C
  tracks [{type=MovieSceneAudioTrack, display_name=Audio, sections []}]
 }
]
display_rate {denominator=1, numerator=30}
tick_resolution {denominator=1, numerator=24000}
tracks []
