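# Level Sequence that times the audio for the DragonCart rift-out portal animation, from a long buildup into overlapping later sounds.
# Audio sections attached to the front buildup Niagara actor, two rift actors and the closed orb actor layer the sound, several shaped by keyed volume fades.
# A Niagara float track ramps ZapAudioVolume through the buildup.
# Event tracks on the orb and portal actors fire triggers alongside the audio.
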LevelSequence /DragonCartNarrative/Assets/VFX/RiftPortalIdle/AnimatedPortal/S_DragonCart_Portal_RiftOut_Audio
bindings [
 NS_DCN_Rift_Buildup_Front {
  id=#1
  possessed_class=NiagaraActor
  tracks [{
    type=MovieSceneAudioTrack
    display_name=Audio
    sections [
     {
      type=MovieSceneAudioSection
      end_frame=913
      start_frame=2
      channels [
       Volume {type=MovieSceneScriptingFloatChannel, default=0.5, keys [[2,0],[395,0.5],[637,1],[898,1],[911,0]]}
       Pitch {type=MovieSceneScriptingFloatChannel, default=1}
      ]
     }
     {
      type=MovieSceneAudioSection
      end_frame=932
      start_frame=403
      channels [
       Volume {type=MovieSceneScriptingFloatChannel, default=1}
       Pitch {type=MovieSceneScriptingFloatChannel, default=1}
      ]
     }
     {
      type=MovieSceneAudioSection
      end_frame=925
      start_frame=850
      channels [
       Volume {type=MovieSceneScriptingFloatChannel, default=1}
       Pitch {type=MovieSceneScriptingFloatChannel, default=1}
      ]
     }
    ]
   }]
 }
 BGA_DCN_OrbActorClosed {
  id=#2
  possessed_class=BGA_DCN_OrbActor_C
  tracks [
   {
    type=MovieSceneAudioTrack
    display_name=Audio
    sections [
     {
      type=MovieSceneAudioSection
      end_frame=1079
      start_frame=1006
      channels [
       Volume {type=MovieSceneScriptingFloatChannel, default=1}
       Pitch {type=MovieSceneScriptingFloatChannel, default=1}
      ]
     }
     {
      type=MovieSceneAudioSection
      end_frame=1492
      start_frame=991
      channels [
       Volume {type=MovieSceneScriptingFloatChannel, default=0.8, keys [[991.8075,0.8],[1003,0.75],[1209,1],[1305,0]]}
       Pitch {type=MovieSceneScriptingFloatChannel, default=1}
      ]
     }
     {
      type=MovieSceneAudioSection
      end_frame=1599
      start_frame=945
      channels [
       Volume {type=MovieSceneScriptingFloatChannel, default=1}
       Pitch {type=MovieSceneScriptingFloatChannel, default=1}
      ]
     }
    ]
   }
   {
    type=MovieSceneEventTrack
    sections [{
      type=MovieSceneEventTriggerSection
      channels [None {
        type=MovieSceneScriptingEventChannel
        keys [[945, "<Struct 'MovieSceneEvent' (0x49ddd17dc00) {}>"], [1239, "<Struct 'MovieSceneEvent' (0x49ddd17e560) {}>"]]
       }]
     }]
   }
  ]
 }
 NiagaraComponent0 {
  id=#3
  parent_id=#1
  possessed_class=NiagaraComponent
  tracks [{
    type=MovieSceneNiagaraFloatParameterTrack
    display_name=ZapAudioVolume
    sections [{type=MovieSceneFloatSection, channels [None {type=MovieSceneScriptingFloatChannel, default=0.8, keys [[0,0],[749,0.8],[905,1]]}]}]
   }]
 }
 SM_DragonCart_Portal {
  id=#4
  possessed_class=NiagaraActor
  tracks [{
    type=MovieSceneEventTrack
    sections [{
      type=MovieSceneEventTriggerSection
      channels [None {type=MovieSceneScriptingEventChannel, const ["<Struct 'MovieSceneEvent' (0x49ddd17e6a0) {}>", 1]}]
     }]
   }]
 }
 BP_DCN_Rift_Actor2 {
  id=#5
  possessed_class=BP_DCN_Rift_Actor_C
  tracks [{
    type=MovieSceneAudioTrack
    display_name=Audio
    sections [
     {
      type=MovieSceneAudioSection
      end_frame=1087
      start_frame=916
      channels [
       Volume {type=MovieSceneScriptingFloatChannel, default=1}
       Pitch {type=MovieSceneScriptingFloatChannel, default=1}
      ]
     }
     {
      type=MovieSceneAudioSection
      end_frame=1288
      start_frame=1178
      channels [
       Volume {type=MovieSceneScriptingFloatChannel, default=0, keys [[1178,0],[1193,1],[1268,1],[1288,0]]}
       Pitch {type=MovieSceneScriptingFloatChannel, default=1}
      ]
     }
     {
      type=MovieSceneAudioSection
      end_frame=1405
      start_frame=1139
      channels [
       Volume {type=MovieSceneScriptingFloatChannel, default=1}
       Pitch {type=MovieSceneScriptingFloatChannel, default=1}
      ]
     }
    ]
   }]
 }
 BP_DCN_Rift_Actor {
  id=#6
  possessed_class=BP_DCN_Rift_Actor_C
  tracks [{
    type=MovieSceneAudioTrack
    display_name=Audio
    sections [
     {
      type=MovieSceneAudioSection
      end_frame=913
      start_frame=2
      channels [
       Volume {type=MovieSceneScriptingFloatChannel, default=0.4, keys [[2,0],[395,0.4],[637,1],[898,1],[911,0]]}
       Pitch {type=MovieSceneScriptingFloatChannel, default=1}
      ]
     }
     {
      type=MovieSceneAudioSection
      end_frame=929
      start_frame=400
      channels [
       Volume {type=MovieSceneScriptingFloatChannel, default=1}
       Pitch {type=MovieSceneScriptingFloatChannel, default=1}
      ]
     }
     {
      type=MovieSceneAudioSection
      end_frame=923
      start_frame=848
      channels [
       Volume {type=MovieSceneScriptingFloatChannel, default=1}
       Pitch {type=MovieSceneScriptingFloatChannel, default=1}
      ]
     }
     {
      type=MovieSceneAudioSection
      end_frame=1074
      start_frame=903
      channels [
       Volume {type=MovieSceneScriptingFloatChannel, default=1}
       Pitch {type=MovieSceneScriptingFloatChannel, default=1}
      ]
     }
     {
      type=MovieSceneAudioSection
      end_frame=1387
      start_frame=1218
      channels [
       Volume {type=MovieSceneScriptingFloatChannel, default=1}
       Pitch {type=MovieSceneScriptingFloatChannel, default=1}
      ]
     }
     {
      type=MovieSceneAudioSection
      end_frame=1219
      start_frame=1176
      channels [
       Volume {type=MovieSceneScriptingFloatChannel, default=0}
       Pitch {type=MovieSceneScriptingFloatChannel, default=1}
      ]
     }
    ]
   }]
 }
]
display_rate {denominator=1, numerator=30}
tick_resolution {denominator=1, numerator=24000}
tracks []
