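# Level Sequence for the DragonCart rift exit: the rift builds up, a portal opens, and the orb actor rises through it before it closes.
# Spawn tracks stage Niagara buildup, sky glow, explosion light, portal and close effects, with keyed parameters driving glow and buildup.
# An MPC_DragonCart_Orb_Rift track animates portal opening and glow colors and holds the portal position for materials.
# It also carries orb material overrides, a subsequence and an UnloadOrb event.
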LevelSequence /DragonCartNarrative/Assets/VFX/RiftPortalIdle/AnimatedPortal/S_DragonCart_RiftOut
bindings [
 NS_DragonCart_Animated_Portal_Close {
  id=#1
  possessed_class=NiagaraActor
  tracks [
   {
    type=MovieSceneSpawnTrack
    sections [{
      type=MovieSceneSpawnSection
      channels [None {type=MovieSceneScriptingBoolChannel, default=false, keys [[584.0025,false],[714,true],[1103.38,false]]}]
     }]
   }
   {
    type=MovieScene3DTransformTrack
    sections [{
      type=MovieScene3DTransformSection
      channels [
       Location.X {type=MovieSceneScriptingDoubleChannel, default=-10752}
       Location.Y {type=MovieSceneScriptingDoubleChannel, default=0}
       Location.Z {type=MovieSceneScriptingDoubleChannel, default=17813.3003}
       Rotation.X {type=MovieSceneScriptingDoubleChannel, default=0}
       Rotation.Y {type=MovieSceneScriptingDoubleChannel, default=0}
       Rotation.Z {type=MovieSceneScriptingDoubleChannel, default=0}
       Scale.X {type=MovieSceneScriptingDoubleChannel, default=1}
       Scale.Y {type=MovieSceneScriptingDoubleChannel, default=1}
       Scale.Z {type=MovieSceneScriptingDoubleChannel, default=1}
      ]
     }]
   }
  ]
 }
 NiagaraComponent0 {
  id=#2
  parent_id=#3
  possessed_class=NiagaraComponent
  tracks [{
    type=MovieSceneNiagaraFloatParameterTrack
    display_name=Offscreen Glow Alpha
    sections [{
      type=MovieSceneFloatSection
      channels [None {type=MovieSceneScriptingFloatChannel, default=0.4, keys [[-1084,0.1],[-488,0.2],[2,0.2],[62,0.4],[199,0.2],[734,0.2],[791,0]]}]
     }]
   }]
 }
 BP_DCN_Rift_Actor {
  id=#4
  possessed_class=BP_DCN_Rift_Actor_C
  tracks [{
    type=MovieSceneVisibilityTrack
    sections [{type=MovieSceneVisibilitySection, channels [None {type=MovieSceneScriptingBoolChannel, default=true, const [true,1]}]}]
   }]
 }
 SM_DragonCart_Portal {
  id=#5
  possessed_class=NiagaraActor
  tracks [
   {
    type=MovieSceneSpawnTrack
    sections [{
      type=MovieSceneSpawnSection
      channels [None {type=MovieSceneScriptingBoolChannel, default=false, keys [[0,false],[30,true],[830.835,false]]}]
     }]
   }
   {
    type=MovieScene3DTransformTrack
    sections [{
      type=MovieScene3DTransformSection
      channels [
       Location.X {type=MovieSceneScriptingDoubleChannel, default=-10752}
       Location.Y {type=MovieSceneScriptingDoubleChannel, default=0}
       Location.Z {type=MovieSceneScriptingDoubleChannel, default=17813.3003}
       Rotation.X {type=MovieSceneScriptingDoubleChannel, default=0}
       Rotation.Y {type=MovieSceneScriptingDoubleChannel, default=0}
       Rotation.Z {type=MovieSceneScriptingDoubleChannel, default=0}
       Scale.X {type=MovieSceneScriptingDoubleChannel, default=1}
       Scale.Y {type=MovieSceneScriptingDoubleChannel, default=1}
       Scale.Z {type=MovieSceneScriptingDoubleChannel, default=1}
      ]
     }]
   }
  ]
 }
 NS_DCN_Rift_Buildup_Back {
  id=#6
  possessed_class=NiagaraActor
  tracks [
   {
    type=MovieSceneSpawnTrack
    sections [{
      type=MovieSceneSpawnSection
      channels [None {type=MovieSceneScriptingBoolChannel, default=false, keys [[-1795,false],[-1784,true],[80,false]]}]
     }]
   }
   {
    type=MovieScene3DTransformTrack
    sections [{
      type=MovieScene3DTransformSection
      channels [
       Location.X {type=MovieSceneScriptingDoubleChannel, default=-10752}
       Location.Y {type=MovieSceneScriptingDoubleChannel, default=0}
       Location.Z {type=MovieSceneScriptingDoubleChannel, default=17813.3003}
       Rotation.X {type=MovieSceneScriptingDoubleChannel, default=0}
       Rotation.Y {type=MovieSceneScriptingDoubleChannel, default=0}
       Rotation.Z {type=MovieSceneScriptingDoubleChannel, default=0}
       Scale.X {type=MovieSceneScriptingDoubleChannel, default=1}
       Scale.Y {type=MovieSceneScriptingDoubleChannel, default=1}
       Scale.Z {type=MovieSceneScriptingDoubleChannel, default=1}
      ]
     }]
   }
  ]
 }
 NS_DCN_Rift_Buildup_Front {
  id=#7
  possessed_class=NiagaraActor
  tracks [
   {
    type=MovieSceneSpawnTrack
    sections [{
      type=MovieSceneSpawnSection
      channels [None {type=MovieSceneScriptingBoolChannel, default=false, keys [[-1795,false],[-1783.585,true],[79,false]]}]
     }]
   }
   {
    type=MovieScene3DTransformTrack
    sections [{
      type=MovieScene3DTransformSection
      channels [
       Location.X {type=MovieSceneScriptingDoubleChannel, default=-10752}
       Location.Y {type=MovieSceneScriptingDoubleChannel, default=0}
       Location.Z {type=MovieSceneScriptingDoubleChannel, default=17813.3003}
       Rotation.X {type=MovieSceneScriptingDoubleChannel, default=0}
       Rotation.Y {type=MovieSceneScriptingDoubleChannel, default=0}
       Rotation.Z {type=MovieSceneScriptingDoubleChannel, default=0}
       Scale.X {type=MovieSceneScriptingDoubleChannel, default=1}
       Scale.Y {type=MovieSceneScriptingDoubleChannel, default=1}
       Scale.Z {type=MovieSceneScriptingDoubleChannel, default=1}
      ]
     }]
   }
  ]
 }
 NS_DragonCart_SkyGlow {
  id=#3
  possessed_class=NiagaraActor
  tracks [
   {
    type=MovieSceneSpawnTrack
    sections [{
      type=MovieSceneSpawnSection
      channels [None {type=MovieSceneScriptingBoolChannel, default=false, keys [[-1091,false],[-1079.585,true],[794,false]]}]
     }]
   }
   {
    type=MovieScene3DTransformTrack
    sections [{
      type=MovieScene3DTransformSection
      channels [
       Location.X {type=MovieSceneScriptingDoubleChannel, default=-10752}
       Location.Y {type=MovieSceneScriptingDoubleChannel, default=0}
       Location.Z {type=MovieSceneScriptingDoubleChannel, default=17813.3003}
       Rotation.X {type=MovieSceneScriptingDoubleChannel, default=0}
       Rotation.Y {type=MovieSceneScriptingDoubleChannel, default=0}
       Rotation.Z {type=MovieSceneScriptingDoubleChannel, default=0}
       Scale.X {type=MovieSceneScriptingDoubleChannel, default=1}
       Scale.Y {type=MovieSceneScriptingDoubleChannel, default=1}
       Scale.Z {type=MovieSceneScriptingDoubleChannel, default=1}
      ]
     }]
   }
  ]
 }
 NS_DCN_ExplosionLight {
  id=#8
  possessed_class=NiagaraActor
  tracks [
   {
    type=MovieSceneSpawnTrack
    sections [{type=MovieSceneSpawnSection, channels [None {type=MovieSceneScriptingBoolChannel, default=false, keys [[-80,false],[30,true],[209,false]]}]}]
   }
   {
    type=MovieScene3DTransformTrack
    sections [{
      type=MovieScene3DTransformSection
      channels [
       Location.X {type=MovieSceneScriptingDoubleChannel, default=-10752.0001}
       Location.Y {type=MovieSceneScriptingDoubleChannel, default=-5632}
       Location.Z {type=MovieSceneScriptingDoubleChannel, default=10849.3003}
       Rotation.X {type=MovieSceneScriptingDoubleChannel, default=0}
       Rotation.Y {type=MovieSceneScriptingDoubleChannel, default=0}
       Rotation.Z {type=MovieSceneScriptingDoubleChannel, default=0}
       Scale.X {type=MovieSceneScriptingDoubleChannel, default=1}
       Scale.Y {type=MovieSceneScriptingDoubleChannel, default=1}
       Scale.Z {type=MovieSceneScriptingDoubleChannel, default=1}
      ]
     }]
   }
  ]
 }
 OrbHatch {
  id=#9
  parent_id=#10
  possessed_class=StaticMeshComponent
  tracks [{
    type=MovieScenePrimitiveMaterialTrack
    display_name=Material Slot: two
    sections [{
      type=MovieScenePrimitiveMaterialSection
      channels [None {
        type=MovieSceneScriptingObjectPathChannel
        default="<Object '/DragonCartNarrative/Assets/FlyingOrb/Materials/MI_FlyingOrb_A_Masked_Inst' (0x49de44b4800) Class 'MaterialInstanceConstant'>"
       }]
     }]
   }]
 }
 Niagara {
  id=#11
  parent_id=#4
  possessed_class=NiagaraComponent
  tracks [
   {
    type=MovieSceneNiagaraFloatParameterTrack
    display_name=BuildUpIntensity
    sections [{type=MovieSceneFloatSection, channels [None {type=MovieSceneScriptingFloatChannel, default=1.5, keys [[-978,1],[30,2],[594,1.5],[915,1]]}]}]
   }
   {
    type=MovieSceneNiagaraFloatParameterTrack
    display_name=BuildUpScale
    sections [{
      type=MovieSceneFloatSection
      channels [None {
        type=MovieSceneScriptingFloatChannel
        default=0
        keys [[-1799,1],[-1622,0.9633],[-1531,1.0005],[-1478,1.0063],[-1436,0.9455],[-1242,0.9754],[-1192,0.9157],[-983,1.0015],[-885,-0.0035],[593,0],[668,1]]
       }]
     }]
   }
   {
    type=MovieSceneNiagaraFloatParameterTrack
    display_name=AnimateScale
    sections [{
      type=MovieSceneFloatSection
      channels [None {type=MovieSceneScriptingFloatChannel, default=1, keys [[-1795,1],[-1726,0.9],[-244,1.4],[30,1],[594,1.3],[888,1]]}]
     }]
   }
  ]
 }
 BGA_DCN_OrbActorClosed {
  id=#10
  possessed_class=BGA_DCN_OrbActor_C
  tracks [
   {
    type=MovieSceneVisibilityTrack
    sections [{type=MovieSceneVisibilitySection, channels [None {type=MovieSceneScriptingBoolChannel, default=false, keys [[0,true],[794,false]]}]}]
   }
   {
    type=MovieScene3DTransformTrack
    sections [{
      type=MovieScene3DTransformSection
      channels [
       Location.X {type=MovieSceneScriptingDoubleChannel, default=-10752, const [-10752,5]}
       Location.Y {type=MovieSceneScriptingDoubleChannel, default=0, const [0,1]}
       Location.Z {
        type=MovieSceneScriptingDoubleChannel
        default=10749.9258
        keys [[39,11136],[180,11136],[246,10975.448],[550,18981.7227],[1350,25973.1506],[1399,25973.1504]]
       }
       Rotation.X {type=MovieSceneScriptingDoubleChannel, default=0, thinned_from=6, keys [[39,0],[243,0],[367,-0.4819],[454,1.8266],[595,-1.8786]]}
       Rotation.Y {type=MovieSceneScriptingDoubleChannel, default=0, thinned_from=7, keys [[39,0],[113,-0.004],[294,-0.8243],[450,1.5204],[595,0],[1399,0]]}
       Rotation.Z {type=MovieSceneScriptingDoubleChannel, default=0, thinned_from=5, keys [[39,90],[550,90],[1142,0],[1399,0]]}
       Scale.X {type=MovieSceneScriptingDoubleChannel, default=0.75, const [0.75,5]}
       Scale.Y {type=MovieSceneScriptingDoubleChannel, default=0.75, const [0.75,5]}
       Scale.Z {type=MovieSceneScriptingDoubleChannel, default=0.75, const [0.75,5]}
      ]
     }]
   }
  ]
 }
 StaticMesh {
  id=#12
  parent_id=#10
  possessed_class=StaticMeshComponent
  tracks [
   {
    type=MovieScenePrimitiveMaterialTrack
    display_name=Material Slot: one
    sections [{
      type=MovieScenePrimitiveMaterialSection
      channels [None {
        type=MovieSceneScriptingObjectPathChannel
        default="<Object '/DragonCartNarrative/Assets/FlyingOrb/Materials/M_FlyingOrbB_Masked_Inst' (0x49de44b3800) Class 'MaterialInstanceConstant'>"
        const ["<Object '&M_FlyingOrbB_Masked_Inst' (0x49de44b3800) Class 'MaterialInstanceConstant'>", 1]
       }]
     }]
   }
   {
    type=MovieScenePrimitiveMaterialTrack
    display_name=Material Slot: two
    sections [{
      type=MovieScenePrimitiveMaterialSection
      channels [None {
        type=MovieSceneScriptingObjectPathChannel
        default="<Object '&MI_FlyingOrb_A_Masked_Inst' (0x49de44b4800) Class 'MaterialInstanceConstant'>"
        const ["<Object '&MI_FlyingOrb_A_Masked_Inst' (0x49de44b4800) Class 'MaterialInstanceConstant'>", 1]
       }]
     }]
   }
  ]
 }
 NiagaraComponent0 {
  id=#13
  parent_id=#5
  possessed_class=NiagaraComponent
  tracks [{
    type=MovieSceneParticleTrack
    display_name=FX System
    sections [{
      type=MovieSceneParticleSection
      channels [None {type=MovieSceneScriptingParticleChannel, keys [[30, "<ParticleKey.ACTIVATE: 0>"], [720, "<ParticleKey.DEACTIVATE: 1>"]]}]
     }]
   }]
 }
]
display_rate {denominator=1, numerator=60}
tick_resolution {denominator=1, numerator=24000}
tracks [
 {
  type=MovieSceneMaterialParameterCollectionTrack
  display_name=MPC_DragonCart_Orb_Rift
  sections [{
    type=MovieSceneParameterSection
    channels [
     PortalOpen {type=MovieSceneScriptingFloatChannel, default=0.1, keys [[20.8475,0],[30,0.1],[150,1],[570,1],[720,0.1]]}
     InnerGlowRiftScale {type=MovieSceneScriptingFloatChannel, default=1, keys [[-978,2.5],[30,1],[1000,2.5]]}
     PortalPositionR {type=MovieSceneScriptingFloatChannel, default=-10752, const [-10752,1]}
     PortalPositionG {type=MovieSceneScriptingFloatChannel, default=0, const [0,1]}
     PortalPositionB {type=MovieSceneScriptingFloatChannel, const [17813,1]}
     PortalPositionA {type=MovieSceneScriptingFloatChannel, const [0,1]}
     InnerCrackGlowR {type=MovieSceneScriptingFloatChannel, default=2, keys [[-1795,0.5365],[-978,0.5365],[30,2],[1447.9975,0.5365]]}
     InnerCrackGlowG {type=MovieSceneScriptingFloatChannel, default=0.61, keys [[-1795,0.1637],[-978,0.1637],[30,0.61],[1447.9975,0.1637]]}
     InnerCrackGlowB {type=MovieSceneScriptingFloatChannel, default=0.009, keys [[-1795,0.0024],[-978,0.0024],[30,0.009],[1447.9975,0.0024]]}
     InnerCrackGlowA {type=MovieSceneScriptingFloatChannel, const [1,4]}
     InnerRiftGlowR {type=MovieSceneScriptingFloatChannel, default=2, keys [[-1795,0.974],[-978,0.974],[30,2],[1447.9975,0.974]]}
     InnerRiftGlowG {type=MovieSceneScriptingFloatChannel, default=0.5, keys [[-1795,0.2973],[-978,0.2973],[30,0.5],[1447.9975,0.2973]]}
     InnerRiftGlowB {type=MovieSceneScriptingFloatChannel, default=0.022, keys [[-1795,0.0043],[-978,0.0043],[30,0.022],[1447.9975,0.0043]]}
     InnerRiftGlowA {type=MovieSceneScriptingFloatChannel, default=1, const [1,4]}
    ]
   }]
 }
 {type=MovieSceneSubTrack, display_name=Subsequences, sections [{type=MovieSceneSubSection, end_frame=1468, start_frame=-1800}]}
 {
  type=MovieSceneEventTrack
  display_name=UnloadOrb
  sections [{
    type=MovieSceneEventTriggerSection
    channels [None {type=MovieSceneScriptingEventChannel, const ["<Struct 'MovieSceneEvent' (0x49ddd69c300) {}>", 1]}]
   }]
 }
]
