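# Translucent unlit, two-sided mesh-particle material for the DragonCart rift portal: a cracked opening onto a nebula-and-stars interior.
# Background layers combine camera-relative triplanar projections, reflection-vector UVs, a flipbook and bump-offset parallax.
# Crack masks shape opacity with depth and near-camera fades, while MPC_DragonCart_Orb_Rift drives inner glow color and falloff.
# A dynamic "Open" parameter supplies opening animation, alongside switchable vibrating geode-crack WPO.
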Material /DragonCartNarrative/Assets/VFX/RiftPortalIdle/M_DragonCart_Geode_Radial_Warp_Mask
graph {
 properties {
  bUsedWithMeshParticles=true
  bUsedWithNiagaraMeshParticles=true
  blendMode=BLEND_Translucent
  pixelDepthOffsetMode=PDOM_Legacy
  shadingModel=MSM_Unlit
  twoSided=true
  usedShadingModels=Unlit
 }
 nodes [
  n1 = Subtract(
   A=WorldPosition().XYZ
   B=Multiply(A=CameraPositionWS(), B=ScalarParameter() {defaultValue=0.2, group=Background, parameterName=CameraFollowStrenght})
  )
  n2 = ScalarParameter() {defaultValue=0.0001, group=Background, parameterName="BG Trip[lanar"}
  n3 = MaterialFunctionCall!(
   CameraVector=@n1
   Texture=TextureObject() {texture=/Game/Characters/Player/Male/Medium/Bodies/M_Med_Soldier_04/Skins/Galaxy/Textures/T_FN_Nebula_style2}
   Tiling=@n2
   Offset
   AxisFadeContrast
  ) {materialFunction=/Engine/Functions/Engine_MaterialFunctions01/Texturing/TriplanarCameraVector}
  n4 = TextureObject() {samplerType=SAMPLERTYPE_Masks, texture=/Game/Packages/Fortress_Sky/Textures/T_Sky_Stars_Sharp}
  n5 = Time()
  n6 = Multiply(A=Constant3Vector() {constant=<rgba 0.003 0 0 0>}, B=@n5)
  n7 = Saturate(None=Sine(None=@n5) {period=3})
  Add(
   A=@n3.Result
   B=Multiply(
    A=MaterialFunctionCall(CameraVector=@n1, Texture=@n4, Tiling=@n2, Offset=@n6) {materialFunction=&TriplanarCameraVector}.Result
    B=LinearInterpolate(Alpha=@n7) {constA=0.4, constB=0.7}
   )
  )
  Constant() {r=2}
  n8 = TextureSample(UVs=Multiply(A=NamedRerouteUsage(), B=Constant() {r=10})) {texture=/Engine/MapTemplates/Sky/T_Sky_Stars}
  n9 = Constant() {r=0.5}
  n10 = TextureSample(UVs=Multiply(A=NamedRerouteUsage(), B=Constant() {r=3})) {texture=/Game/Characters/Player/Male/Medium/Bodies/M_Med_Soldier_04/Skins/Galaxy/Textures/T_FN_Nebula}
  n11 = Constant() {r=0.5}
  n12 = Add(A=Saturate(None=Multiply(A=@n8.R, B=@n9) {constB=3}), B=Saturate(None=Multiply(A=@n10.RGB, B=@n11)))
  n13 = TextureObjectParameter() {
   group=Background
   parameterName=Rift BG
   texture=/Game/Characters/Player/Male/Medium/Bodies/M_Med_Soldier_04/Skins/Galaxy/Textures/T_FN_Nebula_style_blue
  }
  n14 = TextureCoordinate() {coordinateIndex=1}
  n15 = Constant()
  n16 = Add(A=@n15) {constB=-0.25}
  n17 = Add(
   A=MaterialFunctionCall(
    UVs=MaterialFunctionCall(UVs=@n14, "Rotation Angle (0-1)"=@n16) {materialFunction=&CustomRotator}.Rotated Values
    Texture Scale=Constant() {r=0.3}
   ) {materialFunction=&ScaleUVsByCenter}.UVs
   B=ComponentMask(None=VectorParameter() {group=Background, parameterName=BackgroundUVOffset}.RGB) {g=true, r=true}
  )
  n18 = Constant2Vector() {g=2, r=2}
  n19 = Constant() {r=2}
  n20 = BumpOffset(Coordinate=@n14, Height=Constant() {r=-1})
  n21 = SphereMask(A=QualitySwitch(Default=@n14, Low=@n14, High=@n20, Medium=@n14, Epic=@n20), B=Constant() {r=0.5}) {attenuationRadius=0.15, hardnessPercent=0}
  LinearInterpolate(
   A=LinearInterpolate(
    A=@n12
    B=MaterialFunctionCall!(Texture=@n13, UVs=@n17, SubImages=@n18, Frame=@n19) {materialFunction=/Game/Global/Materials/Character_Parent_2023/VideoFlipbook_Parent_2025/MaterialFunctions/MF_FlipbookSubUV}.RGB
    Alpha=Multiply(A=@n21) {constB=0.82}
   )
   B=@n12
   Alpha=Constant() {r=1}
  )
  n22 = Add(A=NamedRerouteUsage(), B=NamedRerouteUsage())
  MaterialFunctionCall(Emissive=MaterialFunctionCall(LDR Input=Saturate(None=@n22), HDR Tint=@n22) {materialFunction=&DeriveHDRfromLDR}.RemappedRange) {materialFunction=&MF_FX_EmissiveAdjustments}
  n23 = LinearInterpolate(Alpha=NamedRerouteUsage()) {constA=-0.1, constB=0.5}
  n24 = NamedRerouteUsage()
  n25 = TextureSampleParameter2D!(UVs=Multiply(A=NamedRerouteUsage(), B=NamedRerouteUsage()), Bias, Apply View MipBias) {group=RiftCrack, mipValueMode=TMVM_MipBias, parameterName=CrackMask, texture=/Game/Effects/Fort_Effects/Textures/Tile/T_Glass_Cracks_V2_Mask}
  n26 = Add(
   A=SmoothStep(
    Min=@n23
    Max=Add(A=@n23) {constB=0.1}
    Value=TextureSample(UVs=@n24) {samplerSource=SSM_Wrap_WorldGroupSettings, texture=/Game/Effects/Fort_Effects/Textures/Tile/T_Glass_Cracks_M}.R
   )
   B=@n25.G
  )
  n27 = OneMinus(None=StaticSwitchParameter(True=Saturate(None=@n26), False=@n25.G) {group=RiftCrack, parameterName=NewCloseAnim})
  n28 = ScalarParameter() {defaultValue=1, group=Background, parameterName=Opacity Global}
  Multiply(
   A=DepthFade(Opacity=Multiply(A=@n27, B=@n28), FadeDistance=ScalarParameter() {defaultValue=500, group=Background, parameterName=DepthFade})
   B=MaterialFunctionCall(Distance=ScalarParameter() {defaultValue=2000, group=Background, parameterName=NearCameraFade}) {materialFunction=&Near_Camera_Fade}.Result
  )
  n29 = VectorParameter() {group=RiftCrack, parameterName=Tile, defaultValue=<rgba 1.3 1.45 0.5 0.75>}
  NamedRerouteDeclaration(None=AppendVector(A=@n29.B, B=@n29.B)) {name=CrackMaskTileMult, nodeColor=<rgba 0 1 0.9083 1>}
  n30 = ReflectionVectorWS() {bNormalizeCustomWorldNormal=true}
  NamedRerouteDeclaration(None=Multiply(
    A=MaterialFunctionCall!(Vector=@n30) {materialFunction=/Engine/Functions/Engine_MaterialFunctions01/Coordinates/LongLatToUV}.Result
    B=Constant2Vector() {g=1, r=2}
   )) {name=InteriorSpaceUV, nodeColor=<rgba 1 0 0.5156 1>}
  NamedRerouteDeclaration(None=@n3.Result) {name=Background, nodeColor=<rgba 0 0.8359 1 1>}
  NamedRerouteDeclaration(None=DynamicParameter() {paramNames [Param1,Param2,Param3,Open]}.Open) {name=OpenAnim, nodeColor=<rgba 1 0.8438 0 1>}
  n31 = TextureCoordinate() {uTiling=2, vTiling=1.25}
  n32 = AppendVector(A=@n29.R, B=@n29.R)
  n33 = Constant() {r=0.5}
  n34 = LinearInterpolate(
   A=ScalarParameter() {defaultValue=-1.2, group=RiftCrack, parameterName=OpenMin}
   B=ScalarParameter() {defaultValue=-0.5, group=RiftCrack, parameterName=OpenMax}
   Alpha=NamedRerouteUsage()
  ) {constA=-0.73, constB=-0.45}
  NamedRerouteDeclaration(None=Add(A=Multiply(A=@n31, B=@n32), B=AppendVector(A=@n33, B=@n34))) {name=MainUV, nodeColor=<rgba 0 0.8594 1 1>}
  n35 = Multiply(A=NamedRerouteUsage(), B=Add(A=NamedRerouteUsage(), B=Constant2Vector() {g=-0.001}))
  n36 = Power(Base=ComponentMask(None=TextureCoordinate()) {g=true}) {constExponent=12}
  n37 = BumpOffset(Coordinate=@n35, Height=Multiply(A=LinearInterpolate(Alpha=@n36) {constB=1.25}, B=Constant() {r=-2.5}))
  n38 = TextureSampleParameter2D(UVs=QualitySwitch(Default=@n37, Low=@n35, High=@n37, Medium=@n35, Epic=@n37)) {group=RiftCrack, mipValueMode=TMVM_MipBias, parameterName=CrackMask, texture=&T_Glass_Cracks_V2_Mask}
  n39 = ComponentMask(None=CollectionParameter() {collection=&/MPC_DragonCart_Orb_Rift, parameterName=InnerCrackGlow}) {b=true, g=true, r=true}
  n40 = ComponentMask(None=CollectionParameter() {collection=&MPC_DragonCart_Orb_Rift, parameterName=InnerRiftGlow}) {b=true, g=true, r=true}
  n41 = TextureCoordinate() {coordinateIndex=1}
  n42 = Constant() {r=0.5}
  n43 = Power(
   Base=OneMinus(None=SphereMask(A=@n41, B=@n42) {attenuationRadius=0.25, hardnessPercent=1})
   Exp=CollectionParameter() {collection=&MPC_DragonCart_Orb_Rift, parameterName=InnerGlowRiftScale}
  ) {constExponent=2.5}
  NamedRerouteDeclaration(None=Add(A=Multiply(A=@n38.RGB, B=@n39), B=Multiply(A=@n40, B=@n43))) {name=InsetBorderGlow, nodeColor=<rgba 1 0.8203 0 1>}
  OneMinus(None=@n15)
  n44 = TextureCoordinate()
  n45 = NamedRerouteUsage()
  n46 = Multiply(A=Multiply(A=@n45, B=MaterialFunctionCall!(Time=@n45) {materialFunction=/Game/Effects/Fort_Effects/Materials/Geode/_Functions/MF_SineVibration}.Result)) {constB=0.05}
  StaticSwitchParameter(
   True=MaterialFunctionCall!(Mask=ComponentMask(None=@n44) {g=true}, Parameter=Add(A=@n45, B=@n46)) {materialFunction=/Game/Effects/Fort_Effects/Materials/Geode/_Functions/MF_Geode_Crack_WPO}.Result
   False=Constant()
  ) {defaultValue=true, group=RiftCrack, parameterName=UseWPO}
  VectorParameter() {group=Background, parameterName=InnerRiftGlow, defaultValue=<rgba 0.974 0.2973 0.0043 1>}
  VectorParameter() {group=RiftCrack, parameterName=InnerCrackGlow, defaultValue=<rgba 0.5365 0.1637 0.0024 1>}
 ]
}
