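# Niagara system for the idle DragonCart rift portal: crack meshes, layered glow, flickering lightning, and sparks with a center wave.
# Parameter collection NPC_DragonCart_Orb_Rift drives growth through RiftScale curves, while PortalOpen widens the lightning and spark rings, dims glow and switches crack center visibility.
# User parameters tune crack build-up and red glow tint.
# Spawn groups layer each non-lightning emitter; near fade and time-of-day scalars control opacity.
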NiagaraSystem /DragonCartNarrative/Assets/VFX/RiftPortalIdle/NS_DragonCart_Rift_Idle
dependencies {
 usedDataInterfaces [&NiagaraDataInterfaceCurve, &NiagaraDataInterfaceVector2DCurve]
 usedDynamicInputs [
  &Subtract_Float
  &LerpLinearColors
  &FN_LOD_Distance_Remap
 ]
 usedModules [
  &self:SystemSpawnScript.SystemScriptSource.SystemScriptGraph.NiagaraNodeAssignment_0.SetVariables_#1
  &SystemState
  &self:SystemSpawnScript.SystemScriptSource.SystemScriptGraph.NiagaraNodeAssignment_1.SetVariables_#2
  &self:SystemSpawnScript.SystemScriptSource.SystemScriptGraph.NiagaraNodeAssignment_3.SetVariables_#3
 ]
}
emitters {
 Energy_Dark_Glow {
  emitterUpdateScript [
   &EmitterState(Life Cycle Mode=::NewEnumerator1, Loop Behavior=::NewEnumerator1, Loop Duration Mode=::NewEnumerator1, UseLoopDelay=false)
   &SpawnBurst_Instantaneous(Use Spawn Probability=false, UseLoopCountLimit=false) {module=SpawnBurst_Instantaneous002}
   &SpawnBurst_Instantaneous(Spawn Group=1, Use Spawn Probability=false, UseLoopCountLimit=false) {module=SpawnBurst_Instantaneous001}
   &SpawnBurst_Instantaneous(Spawn Group=2, Use Spawn Probability=false, UseLoopCountLimit=false)
  ]
  particleSpawnScript [
   /SetVariables(Particles.VisibilityTag=&SpawnGroup(Max Spawn Group=3)) {isSetParameters=true}
   &InitializeParticle(
    Color=&MakeCustomLinearColorFromBool(
     Bool=$User.RedRift
     True Color=&MergeHelper_ColorSwitch(
      0=<rgba 2 0.6471 0 1>
      1=<rgba 2 0.3478 0.0795 0.8>
      2=<rgba 0.0365 0.0128 0.0128 0.4>
      3=?
      Index=&SpawnGroup()
      Num=::NewEnumerator1
     )
     False Color=&MergeHelper_ColorSwitch(
      0=<rgba 2 0.6471 0 1>
      1=<rgba 2 0.7017 0.1781 0.8>
      2=<rgba 0.0365 0.0128 0.0128 0.4>
      3=?
      Index=&SpawnGroup()
      Num=::NewEnumerator1
     )
    )
    Lifetime=$User.Lifetime
    Mesh Scale Mode=::NewEnumerator3
    Mesh Uniform Scale=&MergeHelper_FloatSwitch(0=0, 1=25, 2=0, 3=?, Index=&SpawnGroup(), Num=::NewEnumerator1)
    Sprite Size Mode=::NewEnumerator3
    Uniform Sprite Size=&MergeHelper_FloatSwitch(0=500, 1=0, 2=4200, 3=?, Index=&SpawnGroup(), Num=::NewEnumerator1)
    UsePositionOffset=false
    Write Material Random=true
   )
   /SetVariables(Particles.DynamicPlasmaEnergy:Vector4f=&MakeVector4(X=0, Y=0, Z=-0.05, W=0)) {isSetParameters=true}
   &InitialMeshOrientation(Mesh Axis To Orient=<0 0 -1>, Mesh Orientation Mode=::NewEnumerator2, Use Rotation Vector=true)
   /SetVariables(Particles.FlareCameraOffset=-100, Particles.FlareDynamic:Vector4f=&MakeVector4(X=1.1, Y=0.8, Z=0, W=0)) {isSetParameters=true}
  ]
  particleUpdateScript [
   &ParticleState(Kill Particles When Lifetime Has Elapsed=false, Let Infinitely Lived Particles Die When Emitter Deactivates=true)
   &ScaleMeshSize(Scale Factor=&VectorFromFloat(Value=&Multiply_Float(
      A=&FloatFromCurve(
       FloatCurve=DataInterface() {propertyValues {Curve {
          defaultValue=0
          keys [
           {arriveTangent=0.3087, interpMode=RCIM_Cubic, leaveTangent=0.3087, tangentMode=RCTM_User, time=0, value=0}
           {arriveTangent=2.242, interpMode=RCIM_Cubic, leaveTangent=2.242, tangentMode=RCTM_User, time=0.1, value=0.27}
           {arriveTangent=0.2412, interpMode=RCIM_Cubic, leaveTangent=0.2412, tangentMode=RCTM_User, time=0.3607, value=0.5722}
           {interpMode=RCIM_Cubic, tangentMode=RCTM_User, time=1, value=1}
          ]
         }}}
       CurveIndex=$NPC.NPC_DragonCart_Orb_Rift.RiftScale
       Scale Curve=&Lerp_Float(A=1, B=3, Alpha=$NPC.NPC_DragonCart_Orb_Rift.PortalOpen)
      )
      B=$System.Scale
     )))
   &ScaleSpriteSize(
    Scale Sprite Size Mode=::NewEnumerator0
    Uniform Scale Factor=&MergeHelper_FloatSwitch(
     0=&Multiply_Float(
      A=&FloatFromCurve(
       FloatCurve=DataInterface() {propertyValues {Curve {
          defaultValue=1
          keys [
           {
            arriveTangent=-0.0852
            arriveTangentWeight=0.0675
            interpMode=RCIM_Cubic
            leaveTangent=-0.0852
            leaveTangentWeight=0.0675
            tangentMode=RCTM_User
            tangentWeightMode=RCTWM_WeightedBoth
            time=0
            value=0
           }
           {arriveTangent=4.0623, interpMode=RCIM_Cubic, leaveTangent=4.0623, tangentMode=RCTM_User, time=0.1, value=0.7}
           {arriveTangent=-0.7778, interpMode=RCIM_Cubic, leaveTangent=-0.7778, time=0.2942, value=1.0401}
           {
            arriveTangent=0.1226
            arriveTangentWeight=0.6888
            interpMode=RCIM_Cubic
            leaveTangent=0.1226
            leaveTangentWeight=0.6888
            tangentMode=RCTM_User
            tangentWeightMode=RCTWM_WeightedBoth
            time=1
            value=0
           }
          ]
         }}}
       CurveIndex=$NPC.NPC_DragonCart_Orb_Rift.RiftScale
      )
      B=$System.Scale
     )
     1=0
     2=&Multiply_Float(A=$NPC.NPC_DragonCart_Orb_Rift.RiftScale, B=$System.Scale)
     3=?
     Index=&SpawnGroup()
     Num=::NewEnumerator1
    )
   )
   &ScaleColor(
    Scale Alpha=&Multiply_Float(
     A=&MergeHelper_FloatSwitch(
      0=&Multiply_Float(
       A=$System.NearFade
       B=&FNTimeOfDayScalar(Night Scale=0.5, Morning Scale=0.7, Day Scale=0.8, Evening Scale=0.6)
      )
      1=&Multiply_Float(
       A=$System.NearFade
       B=&FNTimeOfDayScalar(Night Scale=0.4, Morning Scale=0.6, Day Scale=0.7, Evening Scale=0.5)
      )
      2=$System.NearFade
      3=?
      Index=&SpawnGroup()
      Num=::NewEnumerator1
     )
     B=&Lerp_Float(A=1, B=0.25, Alpha=$NPC.NPC_DragonCart_Orb_Rift.PortalOpen)
    )
    ScaleA=true
    ScaleRGB=true
   )
  ]
  rendererClasses [&NiagaraMeshRendererProperties, &NiagaraSpriteRendererProperties]
  renderers [
   {rendererClass=&NiagaraMeshRendererProperties, rendererIndex=0}
   {rendererClass=&NiagaraSpriteRendererProperties, rendererIndex=1}
   {rendererClass=&NiagaraSpriteRendererProperties, rendererIndex=2}
  ]
 }
 Lightning {
  emitterUpdateScript [
   &EmitterState(
    Enable Distance Culling=true
    Life Cycle Mode=::NewEnumerator1
    Loop Duration=&RandomRangeFloat(Minimum=0.3, Maximum=2)
    MaxDistance=20000
    Module.bUseMaxDistance=true
    Module.bUseMinDistance=false
    Recalculate Duration Each Loop=true
    Scalability Mode=::NewEnumerator1
    UseLoopDelay=false
   )
   &SpawnBurst_Instantaneous(
    Spawn Count=15
    Spawn Time=&MakeCustomFloatFromBool(Bool=&SetBoolByFloatComparison(A=1.5, B=$User.BuildUpIntensity), True Float=1, False Float=0)
    Use Spawn Probability=false
    UseLoopCountLimit=false
   ) {enabled=false, module=SpawnBurst_Instantaneous003}
   &SpawnBurst_Instantaneous(Spawn Probability=0.2, Use Spawn Probability=true, UseLoopCountLimit=false)
   &SpawnBurst_Instantaneous(Spawn Probability=0.25, Spawn Time=0.45, Use Spawn Probability=true, UseLoopCountLimit=false) {module=SpawnBurst_Instantaneous001}
   &SpawnBurst_Instantaneous(Spawn Probability=0.3, Spawn Time=0.75, Use Spawn Probability=true, UseLoopCountLimit=false) {module=SpawnBurst_Instantaneous002}
   &SpawnRate(
    Spawn Probability=0.4
    SpawnRate=&FloatFromCurve(
     FloatCurve=DataInterface() {propertyValues {Curve {keys [
         {arriveTangent=10.4941, interpMode=RCIM_Cubic, time=0, value=0}
         {arriveTangent=5.4555, interpMode=RCIM_Cubic, leaveTangent=5.4555, tangentMode=RCTM_User, time=0.2, value=0.35}
         {
          arriveTangent=-8.7693
          arriveTangentWeight=0.0471
          interpMode=RCIM_Cubic
          leaveTangent=-8.7693
          leaveTangentWeight=0.0707
          tangentMode=RCTM_User
          time=0.3
          value=1
         }
         {arriveTangentWeight=0.3941, interpMode=RCIM_Cubic, leaveTangentWeight=0.9853, tangentMode=RCTM_User, time=0.4, value=0.1}
         <at 1 0.1>
        ]}}}
     CurveIndex=$System.NormalizedLoopAge
     Scale Curve=&Lerp_Float(A=3, B=1, Alpha=$NPC.NPC_DragonCart_Orb_Rift.PortalOpen)
    )
    Use Spawn Probability=true
   )
  ]
  particleSpawnScript [
   /SetVariables(Particles.BiasedRandom=&FloatFromCurve(
     FloatCurve=DataInterface() {propertyValues {Curve {keys [
         {interpMode=RCIM_Cubic, tangentMode=RCTM_User, time=0, value=0}
         {arriveTangent=2.5197, interpMode=RCIM_Cubic, leaveTangent=2.5197, tangentMode=RCTM_User, time=0.6, value=0.6}
         {interpMode=RCIM_Cubic, tangentMode=RCTM_User, time=1, value=1}
        ]}}}
     CurveIndex=&RandomRangeFloat(Minimum=0, Maximum=1)
    )) {isSetParameters=true}
   /Niagara/Modules/Spawn/Initialization/InitializeParticle(
    Color=<rgba 1 0.9011 0 0.5>
    Lifetime=&Lerp_Float(A=0.07, B=0.4, Alpha=$Particles.BiasedRandom)
    Mesh Scale=&MakeVector(
     X=&RandomRangeFloat(Minimum=1, Maximum=2.5)
     Y=&RandomRangeFloat(Minimum=0.25, Maximum=1.5)
     Z=&RandomRangeFloat(Minimum=0.25, Maximum=1.5)
    )
    Module.Write Color=true
    Module.Write Lifetime=true
    Module.Write Mass=true
    Module.Write Material Random=true
    Module.Write Position=false
    Module.Write Scale=true
    Module.Write SpriteRotation=false
    Module.Write SpriteSize=true
    Position=&SimulationPosition()
    Sprite Size=&Vector2DFromFloat(Value=&RandomRangeFloat(Minimum=300, Maximum=700))
   )
   &InitialMeshOrientation(
    Mesh Axis To Orient=<0 0 1>
    Mesh Orientation Mode=::NewEnumerator2
    Rotation=&RandomRangeVector(Minimum=<0 0 0>, Maximum=<1 0 0>)
    Use Rotation Vector=true
   ) {module=InitialMeshOrientation001}
   /SetVariables(Particles.MeshIndex=&RandomRangeInt(Minimum=0, Maximum=1)) {isSetParameters=true}
   /SetVariables(Particles.DynParamsSprite:Vector4f=<0.25 2 1 1>) {isSetParameters=true}
   /SetVariables(Particles.CamOffsetSprite=50) {isSetParameters=true}
   &ShapeLocation(
    Cylinder Distribution=::NewEnumerator2
    Cylinder Height=10
    Cylinder Radius=&RandomRangeFloat(Minimum=900, Maximum=1000)
    MaskSpawnGroup=false
    Non Uniform Scale=&Multiply_VectorByFloat(
     Vector=<1 1 0.3>
     Float=&Lerp_Float(
      A=&FloatFromCurve(
       FloatCurve=DataInterface() {propertyValues {Curve {
          defaultValue=0
          keys [
           {arriveTangent=0.3087, interpMode=RCIM_Cubic, leaveTangent=0.3087, tangentMode=RCTM_User, time=0, value=0}
           {arriveTangent=2.242, interpMode=RCIM_Cubic, leaveTangent=2.242, tangentMode=RCTM_User, time=0.1, value=0.27}
           {arriveTangent=0.2412, interpMode=RCIM_Cubic, leaveTangent=0.2412, tangentMode=RCTM_User, time=0.3607, value=0.5722}
           {interpMode=RCIM_Cubic, tangentMode=RCTM_User, time=1, value=1}
          ]
         }}}
       CurveIndex=$NPC.NPC_DragonCart_Orb_Rift.RiftScale
       Scale Curve=$System.NearFade
      )
      B=5
      Alpha=$NPC.NPC_DragonCart_Orb_Rift.PortalOpen
     )
    )
    Shape Primitive=::NewEnumerator1
    U Position=&RandomRangeFloat(Minimum=0, Maximum=1)
    UseOwnerScale=true
   )
  ]
  particleUpdateScript [
   &ParticleState()
   &ScaleMeshSize(Scale Factor=&VectorFromFloat(Value=&Multiply_Float(A=$System.NearFade, B=$NPC.NPC_DragonCart_Orb_Rift.RiftScale)))
   &ScaleColor(
    Scale Alpha=&FloatFromCurve(
     FloatCurve=DataInterface() {propertyValues {Curve {keys [
         <at 0 0>
         {interpMode=RCIM_Cubic, tangentMode=RCTM_User, time=0.1258, value=0.1952}
         <at 0.341 0.0902>
         {interpMode=RCIM_Cubic, tangentMode=RCTM_User, time=0.496, value=0.2532}
         <at 1 0>
        ]}}}
     Scale Curve=&Multiply_Float(
      A=$System.NearFade
      B=&FNTimeOfDayScalar(Night Scale=0.5, Morning Scale=0.7, Day Scale=0.8, Evening Scale=0.6)
     )
    )
    Scale RGB=&VectorFromFloat(Value=&FloatFromCurve(FloatCurve=DataInterface() {propertyValues {Curve {keys [
          {arriveTangent=3.5209, interpMode=RCIM_Cubic, leaveTangent=3.5209, tangentMode=RCTM_User, time=0, value=0.5}
          {arriveTangent=-20.2701, interpMode=RCIM_Cubic, leaveTangent=-20.2701, tangentMode=RCTM_User, time=0.45, value=5}
          <at 1 1>
         ]}}}))
    ScaleA=true
    ScaleRGB=true
   )
   /SetVariables(Particles.SpriteColor=&MultiplyLinearColor(
     Linear Color=$Particles.Color
     Scale Factor=&MakeVector4_FromLinearColor(LinearColor=&MultiplyLinearColor(
       Linear Color=<rgba 25 15.0635 2.969 1>
       Scale Factor=&MakeVector4(
        X=1
        Y=1
        Z=1
        W=&FloatFromCurve(
         FloatCurve=DataInterface() {propertyValues {Curve {keys [
             {interpMode=RCIM_Cubic, tangentMode=RCTM_User, time=0, value=1}
             {arriveTangentWeight=0.031, interpMode=RCIM_Cubic, leaveTangentWeight=0.0366, tangentMode=RCTM_User, time=0.0918, value=0.05}
             {arriveTangentWeight=0.16, interpMode=RCIM_Cubic, leaveTangent=18.7153, leaveTangentWeight=0.1479, tangentMode=RCTM_Break, time=0.18, value=0}
             {arriveTangent=-0.2815, interpMode=RCIM_Cubic, leaveTangent=-0.2815, tangentMode=RCTM_User, time=0.3, value=0.95}
             {arriveTangentWeight=0.0345, interpMode=RCIM_Cubic, leaveTangentWeight=0.0566, tangentMode=RCTM_User, time=0.4, value=0.1}
             {
              arriveTangentWeight=0.1589
              interpMode=RCIM_Cubic
              leaveTangent=29.6956
              leaveTangentWeight=0.1018
              tangentMode=RCTM_Break
              time=0.5
              value=0.05
             }
             {arriveTangent=0.512, interpMode=RCIM_Cubic, leaveTangent=0.512, tangentMode=RCTM_User, time=0.5641, value=0.8648}
             <at 1 0>
            ]}}}
         Scale Curve=&Multiply_Float(
          A=$System.NearFade
          B=&FNTimeOfDayScalar(Night Scale=0.5, Morning Scale=0.7, Day Scale=0.8, Evening Scale=0.6)
         )
        )
       )
      ))
    )) {isSetParameters=true}
   &DynamicMaterialParameters(
    Index 0 Param 1=&FloatFromCurve(FloatCurve=DataInterface() {propertyValues {Curve {defaultValue=1, keys [{tangentMode=RCTM_User, time=0, value=0}, {tangentMode=RCTM_User, time=1, value=1}]}}})
    Index 0 Param 2=&Add_Float(
     A=&FloatFromCurve(FloatCurve=DataInterface() {propertyValues {Curve {keys [
          {interpMode=RCIM_Cubic, leaveTangent=8.981, tangentMode=RCTM_Break, time=0, value=-0.3}
          {arriveTangent=0.0136, interpMode=RCIM_Cubic, leaveTangent=-5.3647, tangentMode=RCTM_Break, time=0.4833, value=1}
          <at 1 -0.3>
         ]}}})
     B=&Multiply_Float(A=$Particles.BiasedRandom, B=0.5)
    )
    Index 0 Param 3=&RandomRangeFloat(Minimum=0, Maximum=3)
    Index 1 Param 1=&Add_Float(A=$Particles.NormalizedAge, B=&RandomRangeFloat(Minimum=0, Maximum=1))
    Index 1 Param 2=&RandomRangeFloat(Minimum=-50, Maximum=50)
    Index 1 Param 3=3
    Module.Param0WriteEnabled=true
    Module.Param0WriteEnabled1=true
    Module.Param1WriteEnabled=true
    Module.Param1WriteEnabled1=true
    Module.Param2WriteEnabled=true
    Module.Param2WriteEnabled1=true
    Module.Param3WriteEnabled=false
    Module.Param3WriteEnabled1=false
    Write Parameter Index 1=true
   ) {module=DynamicMaterialParameters001}
   &ScaleSpriteSize(
    Uniform Curve Scale=$NPC.NPC_DragonCart_Orb_Rift.RiftScale
    Uniform Curve Sprite Scale=DataInterface() {propertyValues {Curve {keys [
        {arriveTangent=21.5526, interpMode=RCIM_Cubic, time=0, value=0}
        <at 0.1 1 RCIM_Cubic>
        {arriveTangent=-1.8758, interpMode=RCIM_Cubic, leaveTangent=-1.8758, tangentMode=RCTM_User, time=1, value=0}
       ]}}}
   )
  ]
  rendererClasses [&NiagaraMeshRendererProperties, &NiagaraSpriteRendererProperties]
  renderers [
   {rendererClass=&NiagaraMeshRendererProperties, rendererIndex=0}
   {rendererClass=&NiagaraSpriteRendererProperties, rendererIndex=1}
  ]
 }
 RiftCrack {
  emitterUpdateScript [
   &EmitterState(Life Cycle Mode=::NewEnumerator1, Loop Behavior=::NewEnumerator1, Loop Duration Mode=::NewEnumerator1, UseLoopDelay=false)
   &SpawnBurst_Instantaneous(Use Spawn Probability=false, UseLoopCountLimit=false)
   &SpawnBurst_Instantaneous(Spawn Group=1, Use Spawn Probability=false, UseLoopCountLimit=false) {module=SpawnBurst_Instantaneous001}
  ]
  particleSpawnScript [
   /Niagara/Modules/Spawn/Initialization/V2/InitializeParticle(
    Color Mode=::NewEnumerator0
    Lifetime=$User.Lifetime
    Mesh Scale Mode=::NewEnumerator3
    Mesh Uniform Scale=&MergeHelper_FloatSwitch(0=20, 1=13, 2=?, 3=?, Index=&SpawnGroup(), Num=::NewEnumerator0)
    Position Mode=::NewEnumerator0
    Write Material Random=true
   )
   /SetVariables(Particles.VisibilityTag=&SpawnGroup(Max Spawn Group=3)) {isSetParameters=true}
   &InitialMeshOrientation(Mesh Orientation Mode=::NewEnumerator2, Use Rotation Vector=true)
   &SystemLocation(
    MaskSpawnGroup=false
    Offset=&MergeHelper_VectorSwitch(0=<0 0 -750>, 1=<0 0 -100>, 2=?, 3=?, Index=&SpawnGroup(), Num=::NewEnumerator0)
   ) {enabled=false}
  ]
  particleUpdateScript [
   &ParticleState(Kill Particles When Lifetime Has Elapsed=false, Let Infinitely Lived Particles Die When Emitter Deactivates=true)
   &ScaleMeshSize(Scale Factor=&VectorFromFloat(Value=&MergeHelper_FloatSwitch(
      0=&Multiply_Float(
       A=$System.Scale
       B=&FloatFromCurve(
        FloatCurve=DataInterface() {propertyValues {Curve {
           defaultValue=0
           keys [
            {arriveTangent=0.3087, interpMode=RCIM_Cubic, leaveTangent=0.3087, tangentMode=RCTM_User, time=0, value=0}
            {arriveTangent=2.242, interpMode=RCIM_Cubic, leaveTangent=2.242, tangentMode=RCTM_User, time=0.1, value=0.27}
            {arriveTangent=0.2412, interpMode=RCIM_Cubic, leaveTangent=0.2412, tangentMode=RCTM_User, time=0.3607, value=0.5722}
            {interpMode=RCIM_Cubic, tangentMode=RCTM_User, time=1, value=1}
           ]
          }}}
        CurveIndex=$NPC.NPC_DragonCart_Orb_Rift.RiftScale
        Scale Curve=$User.AnimateScale
       )
      )
      1=&Multiply_Float(
       A=&FloatFromCurve(
        FloatCurve=DataInterface() {propertyValues {Curve {
           defaultValue=0
           keys [
            {interpMode=RCIM_Cubic, tangentMode=RCTM_User, time=0.3, value=0}
            {arriveTangent=1.4286, interpMode=RCIM_Cubic, leaveTangent=1.4286, time=0.5954, value=0.7322}
            {arriveTangent=0.1515, interpMode=RCIM_Cubic, leaveTangent=0.1515, tangentMode=RCTM_User, time=1, value=1}
           ]
          }}}
        CurveIndex=$NPC.NPC_DragonCart_Orb_Rift.RiftScale
       )
       B=&MakeCustomFloatFromBool(
        Bool=&SetBoolByFloatComparison(A=3600, B=$Engine.Owner.LODDistance, Comparison Type=::NewEnumerator2)
        True Float=0
        False Float=$System.Scale
       )
      )
      2=?
      3=?
      Index=&SpawnGroup()
      Num=::NewEnumerator0
     )))
   &DynamicMaterialParameters(
    Index 0 Param 1=1
    Index 0 Param 2=&FloatFromCurve(
     FloatCurve=DataInterface() {propertyValues {Curve {
        defaultValue=0
        keys [
         {arriveTangent=0.3087, interpMode=RCIM_Cubic, leaveTangent=0.3087, tangentMode=RCTM_User, time=0, value=0}
         {arriveTangent=2.242, interpMode=RCIM_Cubic, leaveTangent=2.242, tangentMode=RCTM_User, time=0.1, value=0.27}
         {arriveTangent=0.2412, interpMode=RCIM_Cubic, leaveTangent=0.2412, tangentMode=RCTM_User, time=0.3607, value=0.5722}
         {interpMode=RCIM_Cubic, tangentMode=RCTM_User, time=1, value=1.1}
        ]
       }}}
     CurveIndex=$NPC.NPC_DragonCart_Orb_Rift.RiftScale
     Scale Curve=&ClampFloat(Float=$User.BuildUpIntensity, Max=2)
    )
    Index 0 Param 3=&Multiply_Float(A=$User.BuildUpIntensity, B=5)
    Index 0 Param 4=&FloatFromCurve(
     FloatCurve=DataInterface() {propertyValues {Curve {
        defaultValue=0
        keys [
         {arriveTangent=1.5001, interpMode=RCIM_Cubic, leaveTangent=1.5001, tangentMode=RCTM_User, time=0, value=0.1}
         {arriveTangent=0.2422, interpMode=RCIM_Cubic, leaveTangent=0.2422, tangentMode=RCTM_User, time=1, value=0.7882}
        ]
       }}}
     CurveIndex=$NPC.NPC_DragonCart_Orb_Rift.RiftScale
    )
    Index 3 Param 1=&Multiply_Float(A=$NPC.NPC_DragonCart_Orb_Rift.PortalOpen, B=5400)
    Module.Index0Float0Write=true
    Module.Index0Float1Write=true
    Module.Index0Float2Write=true
    Module.Index0Float3Write=true
    Module.Index1Float0Write=true
    Module.Index1Float1Write=true
    Module.Index1Float2Write=true
    Module.Index1Float3Write=true
    Module.Index3Float0Write=true
    Module.Index3Float1Write=false
    Module.Index3Float2Write=false
    Module.Index3Float3Write=true
    Write Parameter Index 1=::NewEnumerator1
    Write Parameter Index 3=::NewEnumerator1
   )
   &ScaleColor(
    Scale Alpha=&Multiply_Float(
     A=$System.NearFade
     B=&FNTimeOfDayScalar(Night Scale=0.5, Morning Scale=0.8, Evening Scale=0.65)
    )
    Scale RGB=&VectorFromFloat(Value=$User.BuildUpIntensity)
    ScaleA=true
    ScaleRGB=true
   )
   /SetVariables(Particles.CenterPortalDynamic:Vector4f=&MakeVector4(
     X=0
     Y=0
     Z=0
     W=&Multiply_Float(
      A=&FloatFromCurve(
       FloatCurve=DataInterface() {propertyValues {Curve {
          defaultValue=0
          keys [
           {interpMode=RCIM_Cubic, tangentMode=RCTM_User, time=0, value=0}
           {arriveTangent=1.338, interpMode=RCIM_Cubic, leaveTangent=1.338, tangentMode=RCTM_Break, time=0.6263, value=0.6637}
           {
            arriveTangent=0.2063
            arriveTangentWeight=0.3459
            interpMode=RCIM_Cubic
            leaveTangent=0.2063
            leaveTangentWeight=0.3459
            tangentMode=RCTM_User
            tangentWeightMode=RCTWM_WeightedBoth
            time=1.0041
            value=0.9029
           }
          ]
         }}}
       CurveIndex=$NPC.NPC_DragonCart_Orb_Rift.RiftScale
       Scale Curve=$User.BuildUpScale
      )
      B=$System.NearFade
     )
    )) {isSetParameters=true}
   /SetVariables(Particles.CenterVisibility:NiagaraInt32=&MakeCustomIntFromBool(
     Bool=&SetBoolByFloatComparison(A=$NPC.NPC_DragonCart_Orb_Rift.PortalOpen, B=0.1)
     True INT=5
     False INT=&SpawnGroup()
    )) {isSetParameters=true}
  ]
  rendererClasses [&NiagaraMeshRendererProperties]
  renderers [
   {rendererClass=&NiagaraMeshRendererProperties, rendererIndex=0}
   {rendererClass=&NiagaraMeshRendererProperties, rendererIndex=1}
  ]
 }
 Sparks_Center_Wave {
  emitterUpdateScript [
   &EmitterState(
    Enable Distance Culling=true
    Life Cycle Mode=::NewEnumerator1
    Loop Behavior=::NewEnumerator1
    Loop Duration Mode=::NewEnumerator1
    MaxDistance=50000
    Module.bUseMaxDistance=true
    Module.bUseMinDistance=false
    Scalability Mode=::NewEnumerator1
    UseLoopDelay=false
   )
   &SpawnRate(SpawnRate=4, Use Spawn Probability=false)
   &SpawnRate(SpawnGroup=1, SpawnRate=4, Use Spawn Probability=false) {module=SpawnRate002}
   &SpawnRate(
    Spawn Probability=0.7
    SpawnGroup=2
    SpawnRate=&Lerp_Float(A=1, B=0, Alpha=$NPC.NPC_DragonCart_Orb_Rift.PortalOpen)
    Use Spawn Probability=true
   ) {module=SpawnRate001}
  ]
  particleSpawnScript [
   /SetVariables(Particles.VisibilityTag=&SpawnGroup()) {isSetParameters=true}
   &InitializeParticle(
    Color=<rgba 2 0.652 0.0289 1>
    Lifetime=&MergeHelper_FloatSwitch(
     0=&RandomRangeFloat(Minimum=2, Maximum=3)
     1=&RandomRangeFloat(Minimum=1, Maximum=2)
     2=&RandomRangeFloat(Minimum=3, Maximum=4)
     3=?
     Index=&SpawnGroup()
     Num=::NewEnumerator1
    )
    Sprite Rotation Mode=::NewEnumerator2
    Sprite Size Mode=::NewEnumerator3
    Uniform Sprite Size=&MergeHelper_FloatSwitch(
     0=&Multiply_Float(
      A=&RandomRangeFloat(Minimum=50, Maximum=80)
      B=&Multiply_Float(A=$NPC.NPC_DragonCart_Orb_Rift.RiftScale, B=$System.Scale)
     )
     1=&Multiply_Float(
      A=&RandomRangeFloat(Minimum=15, Maximum=25)
      B=&Multiply_Float(A=$NPC.NPC_DragonCart_Orb_Rift.RiftScale, B=$System.Scale)
     )
     2=&Multiply_Float(A=&RandomRangeFloat(Minimum=800, Maximum=950), B=$System.Scale)
     3=?
     Index=&SpawnGroup()
     Num=::NewEnumerator1
    )
    UsePositionOffset=false
    Write Material Random=true
   )
   &ShapeLocation(
    Cylinder Distribution=::NewEnumerator2
    Cylinder Height=10
    Cylinder Radius=&MergeHelper_FloatSwitch(
     0=&RandomRangeFloat(Minimum=950, Maximum=1100)
     1=&RandomRangeFloat(Minimum=950, Maximum=1100)
     2=0
     3=?
     Index=&SpawnGroup()
     Num=::NewEnumerator1
    )
    MaskSpawnGroup=false
    Non Uniform Scale=&Multiply_VectorByFloat(
     Vector=<1 1 0.3>
     Float=&Lerp_Float(
      A=&FloatFromCurve(
       FloatCurve=DataInterface() {propertyValues {Curve {
          defaultValue=0
          keys [
           {arriveTangent=0.3087, interpMode=RCIM_Cubic, leaveTangent=0.3087, tangentMode=RCTM_User, time=0, value=0}
           {arriveTangent=2.242, interpMode=RCIM_Cubic, leaveTangent=2.242, tangentMode=RCTM_User, time=0.1, value=0.27}
           {arriveTangent=0.2412, interpMode=RCIM_Cubic, leaveTangent=0.2412, tangentMode=RCTM_User, time=0.3607, value=0.5722}
           {interpMode=RCIM_Cubic, tangentMode=RCTM_User, time=1, value=1}
          ]
         }}}
       CurveIndex=$NPC.NPC_DragonCart_Orb_Rift.RiftScale
       Scale Curve=$System.NearFade
      )
      B=5
      Alpha=$NPC.NPC_DragonCart_Orb_Rift.PortalOpen
     )
    )
    Shape Primitive=::NewEnumerator1
    U Position=&RandomRangeFloat(Minimum=0, Maximum=1)
    UseOwnerScale=true
   )
   &AddVelocity(
    Module.Use Falloff Distance=false
    Override Seed=false
    Velocity Mode=::NewEnumerator1
    Velocity Speed=&MergeHelper_FloatSwitch(
     0=&RandomRangeFloat(Minimum=200, Maximum=500)
     1=&RandomRangeFloat(Minimum=100, Maximum=400)
     2=0
     3=?
     Index=&SpawnGroup()
     Num=::NewEnumerator1
    )
   )
   /SetVariables(Particles.EnergySpriteFacing:Vector3f=<0 0 1>) {isSetParameters=true}
  ]
  particleUpdateScript [
   &ParticleState()
   &SolveForcesAndVelocity(Clamp Velocity=false, Limit Acceleration=false)
   &ScaleSpriteSize(
    Scale Sprite Size Mode=::NewEnumerator0
    Uniform Scale Factor=&MergeHelper_FloatSwitch(
     0=&FloatFromCurve(
      FloatCurve=DataInterface() {propertyValues {Curve {keys [
          {interpMode=RCIM_Cubic, tangentMode=RCTM_User, time=0, value=0.5}
          {interpMode=RCIM_Cubic, tangentMode=RCTM_User, time=0.0787, value=1}
          {interpMode=RCIM_Cubic, tangentMode=RCTM_Break, time=1, value=0.2}
         ]}}}
      Scale Curve=2
     )
     1=&FloatFromCurve(
      FloatCurve=DataInterface() {propertyValues {Curve {
         defaultValue=0
         keys [
          {arriveTangent=4.1785, interpMode=RCIM_Cubic, leaveTangent=4.1785, tangentMode=RCTM_User, time=0, value=0}
          {arriveTangent=1.1179, interpMode=RCIM_Cubic, leaveTangent=1.1179, tangentMode=RCTM_User, time=0.2042, value=0.6228}
          {arriveTangent=0.0224, interpMode=RCIM_Cubic, leaveTangent=0.0224, tangentMode=RCTM_User, time=1, value=1}
         ]
        }}}
      Scale Curve=2
     )
     2=&FloatFromCurve(
      FloatCurve=DataInterface() {propertyValues {Curve {
         defaultValue=0
         keys [
          {arriveTangent=0.3087, interpMode=RCIM_Cubic, leaveTangent=0.3087, tangentMode=RCTM_User, time=0, value=0}
          {arriveTangent=2.242, interpMode=RCIM_Cubic, leaveTangent=2.242, tangentMode=RCTM_User, time=0.1, value=0.27}
          {arriveTangent=0.2412, interpMode=RCIM_Cubic, leaveTangent=0.2412, tangentMode=RCTM_User, time=0.3607, value=0.5722}
          {interpMode=RCIM_Cubic, tangentMode=RCTM_User, time=1, value=1}
         ]
        }}}
      CurveIndex=$NPC.NPC_DragonCart_Orb_Rift.RiftScale
      Scale Curve=1.2
     )
     3=?
     Index=&SpawnGroup()
     Num=::NewEnumerator1
    )
   )
   &ScaleColor(
    Scale Alpha=&MergeHelper_FloatSwitch(
     0=&FloatFromCurve(
      FloatCurve=DataInterface() {propertyValues {Curve {
         defaultValue=1
         keys [
          {interpMode=RCIM_Cubic, tangentMode=RCTM_User, time=0, value=0}
          {interpMode=RCIM_Cubic, tangentMode=RCTM_User, time=0.211, value=0.7864}
          {tangentMode=RCTM_User, time=1, value=0}
         ]
        }}}
      Scale Curve=&FNTimeOfDayScalar(Night Scale=0.5, Morning Scale=0.7, Day Scale=0.8, Evening Scale=0.6)
     )
     1=&FloatFromCurve(
      FloatCurve=DataInterface() {propertyValues {Curve {
         defaultValue=1
         keys [
          {interpMode=RCIM_Cubic, tangentMode=RCTM_User, time=0, value=0}
          {interpMode=RCIM_Cubic, tangentMode=RCTM_User, time=0.211, value=0.7864}
          {tangentMode=RCTM_User, time=1, value=0}
         ]
        }}}
      Scale Curve=&FNTimeOfDayScalar(Night Scale=0.5, Morning Scale=0.7, Day Scale=0.8, Evening Scale=0.6)
     )
     2=&FloatFromCurve(
      FloatCurve=DataInterface() {propertyValues {Curve {
         defaultValue=1
         keys [
          {interpMode=RCIM_Cubic, tangentMode=RCTM_User, time=0, value=0}
          {arriveTangent=-1.0097, interpMode=RCIM_Cubic, leaveTangent=-1.0097, tangentMode=RCTM_User, time=0.7144, value=0.4996}
          {arriveTangent=-2.3597, tangentMode=RCTM_User, time=1, value=0}
         ]
        }}}
      Scale Curve=&FNTimeOfDayScalar(Night Scale=0.5, Morning Scale=0.7, Day Scale=0.8, Evening Scale=0.6)
     )
     3=?
     Index=&SpawnGroup()
     Num=::NewEnumerator1
    )
    ScaleA=true
    ScaleRGB=true
   )
   &ScaleColor(Scale Alpha=$System.NearFade, ScaleA=true, ScaleRGB=true) {module=ScaleColor001}
   /SetVariables(Particles.RunesSubImageIndex:NiagaraFloat=&FloatFromCurve(
     FloatCurve=DataInterface() {propertyValues {Curve {keys [<at 0 1>,<at 1 0>]}}}
     Scale Curve=&RandomRangeFloat(Minimum=0, Maximum=4)
    )) {isSetParameters=true}
  ]
  rendererClasses [&NiagaraSpriteRendererProperties]
  renderers [
   {rendererClass=&NiagaraSpriteRendererProperties}
   {rendererClass=&NiagaraSpriteRendererProperties, rendererIndex=1}
   {rendererClass=&NiagaraSpriteRendererProperties, rendererIndex=2}
  ]
 }
}
summary {
 systemName=NS_DragonCart_Rift_Idle
 emitters [
  {emitterName=Sparks_Center_Wave, rendererClasses [&NiagaraSpriteRendererProperties]}
  {emitterName=RiftCrack, rendererClasses [&NiagaraMeshRendererProperties]}
  {
   emitterName=Energy_Dark_Glow
   rendererClasses [&NiagaraMeshRendererProperties, &NiagaraSpriteRendererProperties]
  }
  {emitterName=Lightning, rendererClasses [&NiagaraMeshRendererProperties, &NiagaraSpriteRendererProperties]}
 ]
 userVariables [
  User.AnimateScale {type=NiagaraFloat, defaultValue=1, description=""}
  User.BuildUpIntensity {type=NiagaraFloat, defaultValue=1, description=""}
  User.BuildUpScale {type=NiagaraFloat, defaultValue=1, description=""}
  User.Lifetime {type=NiagaraFloat, defaultValue=20, description=""}
  User.RedRift {type=NiagaraBool}
 ]
}
