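# Niagara system for orange lightning-style mesh bursts around the rift portal, built from a single emitter named Minimal.
# Short-lived meshes spawn probabilistically on a cylinder's side surface, with radius driven by RiftScale from the NPC_DragonCart_Orb_Rift parameter collection.
# Each gets random scale, orientation along the surface normal and a camera offset.
# Dynamic material parameters feed a falling lifetime curve and a random value to the material, while alpha ramps up.
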NiagaraSystem /DragonCartNarrative/Assets/VFX/RiftPortalIdle/NS_Lightning_Travel_Example
dependencies {usedDataInterfaces [&NiagaraDataInterfaceCurve], usedDynamicInputs [], usedModules [&SystemState]}
emitters {Minimal {
  emitterUpdateScript [
   &EmitterState()
   &SpawnRate(Spawn Probability=0.7, SpawnRate=5, Use Spawn Probability=true)
  ]
  particleSpawnScript [
   &InitializeParticle(
    Color=<rgba 5 2.6183 0.1123 1>
    Lifetime=&RandomRangeFloat(Minimum=0.3, Maximum=0.5)
    Mesh Scale Max=<30 25 25>
    Mesh Scale Min=<20 10 10>
    Mesh Scale Mode=::NewEnumerator2
    UsePositionOffset=false
    Write Material Random=true
   )
   &ShapeLocation(
    Cylinder Height=50
    Cylinder Radius=&Multiply_Float(A=1100, B=$NPC.NPC_DragonCart_Orb_Rift.RiftScale)
    MaskSpawnGroup=false
    Override Seed=false
    Shape Primitive=::NewEnumerator1
    Surface Only=true
    Use Endcaps In Surface Only Mode=false
    UseOwnerScale=true
   )
   &InitialMeshOrientation(
    Mesh Orientation Mode=::NewEnumerator2
    Orientation Vector=$Particles.ShapeLocation.ShapeNormal
    Rotation=&RandomRangeVector(Minimum=<0 0 0>, Maximum=<1 0 0>)
    Use Rotation Vector=true
   )
   &CameraOffset(Camera Offset Amount=-1500)
  ]
  particleUpdateScript [
   &ParticleState()
   &DynamicMaterialParameters(
    Index 0 Param 1=&FloatFromCurve(FloatCurve=DataInterface() {propertyValues {Curve {keys [
         <at 0 1 RCIM_Cubic>
         {arriveTangent=-1.2958, interpMode=RCIM_Cubic, leaveTangent=-1.2958, tangentMode=RCTM_User, time=0.49, value=0.6}
         {arriveTangent=-1.8899, interpMode=RCIM_Cubic, leaveTangent=-1.8899, tangentMode=RCTM_User, time=0.5, value=0.4}
         <at 1 0>
        ]}}})
    Index 0 Param 2=0.5
    Index 0 Param 4=&RandomRangeFloat(Minimum=0, Maximum=1)
    Module.Index0Float0Write=true
    Module.Index0Float1Write=true
    Module.Index0Float2Write=true
    Module.Index0Float3Write=true
   )
   &ScaleColor(
    Scale Alpha=&FloatFromCurve(FloatCurve=DataInterface() {propertyValues {Curve {
        defaultValue=1
        keys [{arriveTangent=3.0787, interpMode=RCIM_Cubic, time=0, value=0}, {interpMode=RCIM_Cubic, tangentMode=RCTM_User, time=0.75, value=1}]
       }}})
    ScaleA=true
    ScaleRGB=false
   )
  ]
  rendererClasses [&NiagaraMeshRendererProperties]
  renderers [{rendererClass=&NiagaraMeshRendererProperties, rendererIndex=0}]
 }}
summary {
 systemName=NS_Lightning_Travel_Example
 emitters [{emitterName=Minimal, rendererClasses [&NiagaraMeshRendererProperties]}]
 userVariables []
}
